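# One-shot Niagara on-body burst of glowing blue energy for the Blue Cheese male medium character cosmetic.
# Arcs bursts meshes and sprites on random shoulder, leg and shoe bones, using sine-blended per-part offsets to slide along each part.
# BoneFlow_ArcEnergy001 builds short-lived, jittered ribbons around bones through custom partition and BoneFlow modules.
# Decal_Burst adds fading glow sprites on random bones of the same separate ribbon skeletal mesh input.
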NiagaraSystem /BRCosmetics/Characters/Player/Male/Medium/Bodies/M_MED_Blue_Cheese/FX/NS_Blue_Cheese_M_OnBody_Burst
dependencies {
 usedDataInterfaces [
  &NiagaraDataInterfaceArrayFloat
  &NiagaraDataInterfaceVectorCurve
  &NiagaraDataInterfaceCurve
  &NiagaraDataInterfaceColorCurve
 ]
 usedDynamicInputs []
 usedModules [&SystemState]
}
emitters {
 Arcs {
  emitterUpdateScript [
   &EmitterState()
   &SpawnBurst_Instantaneous(Spawn Count=&RandomRangeInt(Minimum=12, Maximum=15), Use Spawn Probability=false)
  ]
  particleSpawnScript [
   /SetVariables(Particles.MeshIndex=&RandomRangeInt(Minimum=0, Maximum=1)) {isSetParameters=true}
   /SetVariables(Particles.RandomBone:NiagaraInt32=&RandomRangeInt(Minimum=0, Maximum=7)) {isSetParameters=true}
   &InitializeParticle(
    Lifetime Max=0.8
    Lifetime Min=0.4
    Lifetime Mode=::NewEnumerator1
    Mesh Scale Max=<0.4 0.2 1.5>
    Mesh Scale Min=<0.2 0.1 0.25>
    Mesh Scale Mode=::NewEnumerator2
    Sprite Rotation Mode=::NewEnumerator2
    Sprite Size Mode=::NewEnumerator1
    Uniform Sprite Size Max=20
    Uniform Sprite Size Min=15
    UsePositionOffset=false
    Write Material Random=true
   )
   &InitialMeshOrientation(
    Orientation Vector=&Lerp_Vector(
     A=<1 0 0>
     B=<-1 0 0>
     Alpha=&SelectFloatFromArray(Direct Array Index=$Particles.RandomBone, Float Selection Array=DataInterface() {propertyValues {FloatData [0,1,1,1,1,0,0]}})
    )
    Rotation=&Add_Vector(
     A=&RandomRangeVector(Minimum=<-0.1 0 0>, Maximum=<0.1 0 0>)
     B=&Lerp_Vector(
      A=<0 0 0>
      B=<0.5 0 0>
      Alpha=&SelectFloatFromArray(Direct Array Index=$Particles.RandomBone, Float Selection Array=DataInterface() {propertyValues {FloatData [0,1,1,0,1,0,0]}})
     )
    )
    Use Orientation Vector=true
    Use Rotation Vector=true
   )
   &DynamicMaterialParameters(
    Index 0 Param 2=&RandomRangeFloat(Minimum=1, Maximum=16)
    Index 0 Param 3=1
    Module.Param0WriteEnabled=false
    Module.Param1WriteEnabled=true
    Module.Param2WriteEnabled=true
    Module.Param3WriteEnabled=false
   )
   &AddVelocity(Velocity=&RandomRangeVector(Minimum=<0 0 30>, Maximum=<0 0 20>)) {enabled=false, module=AddVelocity001}
  ]
  particleUpdateScript [
   &ParticleState()
   /SetVariables(Particles.PartBlend:NiagaraFloat=&Sine(
     Normalized Angle=&Add_Float(A=&RandomRangeFloat(Minimum=0, Maximum=1), B=$Emitter.Age)
     Period=&RandomRangeFloat(Minimum=2.5, Maximum=5)
     Scale=0.5
     Bias=0.5
    )) {isSetParameters=true}
   &SkeletalMeshLocation(
    Bone / Socket Index=$Particles.RandomBone
    Bone Sampling Mode=::NewEnumerator4
    MaskSpawnGroup=false
    Sampled Position Offset=&self:Part_Selector(
     Bone Index:NiagaraInt32=$Particles.RandomBone
     Left Calve:Vector3f=&Lerp_Vector(A=<0 -1 -5>, B=<-30 -1 -5>, Alpha=$Particles.PartBlend)
     Left Shoe:Vector3f=&Lerp_Vector(A=<-8 -2 -3>, B=<-8 18 -5>, Alpha=$Particles.PartBlend)
     Left Shoulder:Vector3f=&Lerp_Vector(A=<7 -2 6>, B=<12 -2 6>, Alpha=$Particles.PartBlend)
     Left Thigh:Vector3f=&Lerp_Vector(A=<-10 1 -7>, B=<-35 1 -6>, Alpha=$Particles.PartBlend)
     Right Calve:Vector3f=&Lerp_Vector(A=<0 1 5>, B=<30 1 5>, Alpha=$Particles.PartBlend)
     Right Shoe:Vector3f=&Lerp_Vector(A=<8 -2 3>, B=<8 -18 5>, Alpha=$Particles.PartBlend)
     Right Shoulder:Vector3f=&Lerp_Vector(A=<-7 2 -6>, B=<-12 2 -6>, Alpha=$Particles.PartBlend)
     Right Thigh:Vector3f=&Lerp_Vector(A=<10 -1 7>, B=<35 -1 6>, Alpha=$Particles.PartBlend)
    )
    Skeletal Mesh=$User.SkeletalMesh
   )
   &SolveForcesAndVelocity(Clamp Velocity=false, Limit Acceleration=false)
   &ScaleMeshSize(Scale Factor=&VectorFromCurve(VectorCurve=DataInterface() {propertyValues {
       XCurve {keys [
         {arriveTangent=1.1865, interpMode=RCIM_Cubic, leaveTangent=1.1865, tangentMode=RCTM_User, time=0, value=0.5}
         <at 0.6 1 RCIM_Cubic>
         <at 1 0.8 RCIM_Cubic>
        ]}
       YCurve {keys [{arriveTangent=0.2342, interpMode=RCIM_Cubic, time=0, value=0.9}, <at 0.5 1 RCIM_Cubic>, <at 1 0.9 RCIM_Cubic>]}
       ZCurve {keys [
         {arriveTangent=2.0806, interpMode=RCIM_Cubic, leaveTangent=2.0806, tangentMode=RCTM_User, time=0, value=0.5}
         <at 0.33 1 RCIM_Cubic>
         <at 1 0.5 RCIM_Cubic>
        ]}
      }}))
   &DynamicMaterialParameters(
    Index 0 Param 1=&FloatFromCurve(FloatCurve=DataInterface() {propertyValues {Curve {keys [
         {arriveTangent=9.0725, interpMode=RCIM_Cubic, leaveTangent=10.0491, tangentMode=RCTM_Break, time=0, value=0.5}
         {interpMode=RCIM_Cubic, tangentMode=RCTM_User, time=0.1, value=1}
         {interpMode=RCIM_Cubic, tangentMode=RCTM_User, time=0.5, value=1}
         {interpMode=RCIM_Cubic, leaveTangent=-0.0405, time=1, value=0.1}
        ]}}})
    Index 0 Param 4=1
    Module.Param0WriteEnabled=true
    Module.Param1WriteEnabled=false
    Module.Param2WriteEnabled=false
    Module.Param3WriteEnabled=false
   ) {module=DynamicMaterialParameters001}
   /SetVariables(Particles.Decal DynamicMaterialParam:Vector4f=&MakeVector4(X=1, Y=1, Z=&RandomRangeFloat(Minimum=0, Maximum=100), W=0)) {isSetParameters=true, module=SetVariables_#1}
   &ScaleColor(
    Scale Alpha=&FloatFromCurve(
     FloatCurve=DataInterface() {propertyValues {Curve {keys [
         {arriveTangent=9.0725, interpMode=RCIM_Cubic, leaveTangent=10.0491, tangentMode=RCTM_Break, time=0, value=0}
         {interpMode=RCIM_Cubic, tangentMode=RCTM_User, time=0.1, value=1}
         {interpMode=RCIM_Cubic, tangentMode=RCTM_User, time=0.5, value=1}
         {interpMode=RCIM_Cubic, leaveTangent=-0.0405, time=1, value=0}
        ]}}}
     Scale Curve=0.5
    )
    Scale RGB=&VectorFromFloat(Value=&FloatFromCurve(FloatCurve=DataInterface() {propertyValues {Curve {keys [
          {arriveTangent=9.0725, interpMode=RCIM_Cubic, time=0, value=0.1}
          {interpMode=RCIM_Cubic, tangentMode=RCTM_User, time=0.1, value=1}
          {interpMode=RCIM_Cubic, tangentMode=RCTM_User, time=0.5, value=1}
          {interpMode=RCIM_Cubic, leaveTangent=-0.0405, time=1, value=0.1}
         ]}}}))
    ScaleA=true
    ScaleRGB=true
   )
   &FN_Sprite_Attributes(
    Camera Offset=3
    Color=&MultiplyLinearColor(
     Linear Color=<rgba 0.25 1.8333 5 1>
     Scale Factor=&MakeVector4_FromFloat(Float=&FloatFromCurve(
       FloatCurve=DataInterface() {propertyValues {Curve {keys [
           {arriveTangent=9.0725, interpMode=RCIM_Cubic, leaveTangent=9.0725, tangentMode=RCTM_User, time=0, value=0}
           {arriveTangent=-2.5249, interpMode=RCIM_Cubic, leaveTangent=-2.5249, tangentMode=RCTM_User, time=0.1408, value=0.9847}
           {arriveTangent=-0.0405, interpMode=RCIM_Cubic, leaveTangent=-0.0405, tangentMode=RCTM_User, time=1, value=0}
          ]}}}
       Scale Curve=&RandomRangeFloat(Minimum=1.5, Maximum=2.5)
      ))
    )
    Create Camera Offset Attribute=true
    Create Color Attribute=true
    Write to Module Namespace=true
   )
   &FN_Sprite_Attributes(
    Camera Offset=20
    Color=&MultiplyLinearColor(
     Linear Color=<rgba 0.15 1.1 3 0.05>
     Scale Factor=&MakeVector4_FromFloat(Float=&FloatFromCurve(
       FloatCurve=DataInterface() {propertyValues {Curve {keys [
           {arriveTangent=9.0725, interpMode=RCIM_Cubic, leaveTangent=9.0725, tangentMode=RCTM_User, time=0, value=0}
           {arriveTangent=-2.5249, interpMode=RCIM_Cubic, leaveTangent=-2.5249, tangentMode=RCTM_User, time=0.1408, value=0.9847}
           {arriveTangent=-0.0405, interpMode=RCIM_Cubic, leaveTangent=-0.0405, tangentMode=RCTM_User, time=1, value=0}
          ]}}}
       Scale Curve=&RandomRangeFloat(Minimum=1, Maximum=1)
      ))
    )
    Create Camera Offset Attribute=true
    Create Color Attribute=true
    Create SpriteSize Attribute=true
    SpriteSize=&Multiply_Vector2D_ByFloat(Float=1, Vector2D=$Particles.SpriteSize)
    Write to Module Namespace=true
   ) {module=FN_Sprite_Attributes001}
  ]
  rendererClasses [&NiagaraMeshRendererProperties, &NiagaraSpriteRendererProperties]
  renderers [
   {rendererClass=&NiagaraMeshRendererProperties, rendererIndex=0}
   {rendererClass=&NiagaraSpriteRendererProperties, rendererIndex=1}
   {rendererClass=&NiagaraSpriteRendererProperties, rendererIndex=2}
  ]
 }
 BoneFlow_ArcEnergy001 {
  emitterUpdateScript [
   &EmitterState()
   /SetVariables(Emitter.RandomSeed:NiagaraInt32=&RandomRangeInt(Minimum=0, Maximum=256)) {isSetParameters=true}
   /Game/Effects/Niagara/Modules/Partition/FN_PartitionAttributes(Number of Bone Sets=5, Particles Per Partition=6, Partitions=5)
   &SpawnBurst_Instantaneous(Spawn Count=$Emitter.FN_PartitionAttributes.Spawn Total, Use Spawn Probability=true)
  ]
  particleSpawnScript [
   /Game/Effects/Niagara/Modules/Partition/FN_Particle_partition_v5/FN_Particle_Partition_v5(Enable Ribbon ID=true, Seed:NiagaraInt32=$Emitter.RandomSeed, Shuffle Partition ID=true)
   /SetVariables(Particles.BoneComparison:NiagaraFloat=&MakeCustomFloatFromBool(
     Bool=&SetBoolByIntComparison(A=$Particles.Partition.ID, B=2, Comparison Type=::NewEnumerator6)
     True Float=1
     False Float=0
    )) {isSetParameters=true}
   /Game/Effects/Niagara/Modules/BoneFlow/FN_Initialize_BoneFlow_Attributes_v2(
    Angle:NiagaraFloat=&Add_Float(
     A=&Multiply_Float(
      A=&RandomRangeFloat(Minimum=-0.25, Maximum=0.25)
      B=&Multiply_Float(A=$Particles.Partition.NormalizedID, B=10)
     )
     B=&RandomRangeFloat(Minimum=0.3, Maximum=0.4)
    )
    Radius:NiagaraFloat=&Multiply_Float(
     A=&Lerp_Float(
      A=&RandomRangeFloat(Minimum=14, Maximum=18)
      B=&RandomRangeFloat(Minimum=11, Maximum=12)
      Alpha=$Particles.BoneComparison
     )
     B=&FloatFromCurve(
      FloatCurve=DataInterface() {propertyValues {Curve {keys [
          {interpMode=RCIM_Cubic, tangentMode=RCTM_User, time=0, value=0}
          {interpMode=RCIM_Cubic, tangentMode=RCTM_User, time=0.5, value=1}
          {tangentMode=RCTM_User, time=1, value=0}
         ]}}}
      CurveIndex=$Particles.Partition.NormalizedID
      Scale Curve=0.75
     )
    )
   )
   &InitializeRibbon(
    Color=<rgba 0.25 1.8333 5 1>
    Lifetime=/Game/Effects/Niagara/DynamicInputs/Random/FN_SeededRandomRangeFloat(Minimum:NiagaraFloat=0.4, Maximum:NiagaraFloat=0.6, Random Seed:NiagaraInt32=$Particles.Partition.ID)
    Module.Write Color=true
    Module.Write Lifetime=true
    Module.Write Mass=false
    Module.Write Material Random=true
    Module.Write Position=false
    Module.Write RibbonTwist=false
    Module.Write RibbonWidth=true
    Ribbon Width=10
   )
  ]
  particleUpdateScript [
   &ParticleState()
   &CameraOffset(Camera Offset Amount=&FloatFromCurve(
     FloatCurve=DataInterface() {propertyValues {Curve {keys [
         {arriveTangent=4.3736, interpMode=RCIM_Cubic, time=0, value=0.5}
         {interpMode=RCIM_Cubic, tangentMode=RCTM_User, time=0.5, value=1}
         {interpMode=RCIM_Cubic, leaveTangent=-3.9829, time=1, value=0.5}
        ]}}}
     CurveIndex=$Particles.Partition.NormalizedID
     Scale Curve=20
    )) {enabled=false}
   &ScaleColor(
    Scale Mode=::NewEnumerator1
    Scale RGBA=&Vector4FromCurve(Vector4Curve=DataInterface() {propertyValues {
       AlphaCurve {keys [<at 0.706 1>,<at 1 0>]}
       BlueCurve {keys [<at 0 50>,<at 0.1818 1>,<at 0.4428 1>,<at 0.6162 10>]}
       GreenCurve {keys [<at 0 50>,<at 0.1818 1>,<at 0.4428 1>,<at 0.6162 10>]}
       RedCurve {keys [<at 0 50>,<at 0.1818 1>,<at 0.4428 1>,<at 0.6162 10>]}
      }})
    ScaleRGBA=true
   )
   &DynamicMaterialParameters(
    Index 0 Param 1=1
    Index 0 Param 2=&FloatFromCurve(FloatCurve=DataInterface() {propertyValues {Curve {keys [
         {arriveTangent=16.9415, interpMode=RCIM_Cubic, leaveTangent=16.9415, tangentMode=RCTM_User, time=0, value=-1}
         {arriveTangent=3.3333, leaveTangent=3.3333, time=0.15, value=1}
         {interpMode=RCIM_Cubic, tangentMode=RCTM_User, time=0.5, value=1}
         <at 1 -1 RCIM_Cubic>
        ]}}})
    Index 0 Param 3=&RandomRangeFloat(Minimum=1, Maximum=1.5)
    Index 0 Param 4=&RandomRangeFloat(Minimum=0, Maximum=1)
    Module.Param0WriteEnabled=false
    Module.Param1WriteEnabled=true
    Module.Param2WriteEnabled=true
    Module.Param3WriteEnabled=true
   )
   &FN_BoneFlow(
    Angle:NiagaraFloat=&Multiply_Float(
     A=&RandomRangeFloat(Minimum=-1, Maximum=1)
     B=&Add_Float(
      A=$Particles.Initial.FN_Initialize_BoneFlow_Attributes_v2.Angle
      B=&FloatFromCurve(
       FloatCurve=DataInterface() {propertyValues {Curve {keys [{arriveTangent=6.1308, interpMode=RCIM_Cubic, leaveTangent=6.1308, tangentMode=RCTM_User, time=0, value=0}, <at 1 2 RCIM_Cubic>]}}}
       Scale Curve=0.02
      )
     )
    )
    BoneID:NiagaraInt32=$Particles.Partition.ID
    Height:NiagaraFloat=$Particles.Partition.NormalizedID
    Interpolate Mesh Sampling=false
    Radius:NiagaraFloat=&Add_Float(
     A=$Particles.Initial.FN_Initialize_BoneFlow_Attributes_v2.Radius
     B=&RandomRangeFloat(Minimum=-1, Maximum=1)
    )
    Set Material Dynamic=false
    SkeletalMesh:NiagaraDataInterfaceSkeletalMesh=$User.SkeletalMeshRibbon
   )
  ]
  rendererClasses [&NiagaraRibbonRendererProperties]
  renderers [{rendererClass=&NiagaraRibbonRendererProperties}]
 }
 Decal_Burst {
  emitterSpawnScript [/SetVariables(Emitter.DecalNoiseTile=0.5) {isSetParameters=true}]
  emitterUpdateScript [
   &EmitterState()
   &SpawnBurst_Instantaneous(Spawn Count=16, Use Spawn Probability=false)
  ]
  particleSpawnScript [
   /SetVariables(Particles.RandomBone:NiagaraInt32=&RandomRangeInt(Minimum=0, Maximum=11)) {isSetParameters=true}
   &InitializeParticle(
    Lifetime Max=1
    Lifetime Min=0.75
    Lifetime Mode=::NewEnumerator1
    Mesh Scale Max=<0.4 0.2 1.5>
    Mesh Scale Min=<0.2 0.1 0.25>
    Mesh Scale Mode=::NewEnumerator2
    Sprite Rotation Mode=::NewEnumerator2
    Sprite Size Mode=::NewEnumerator1
    Uniform Sprite Size Max=30
    Uniform Sprite Size Min=25
    UsePositionOffset=false
    Write Material Random=true
   )
   /SetVariables(Particles.Decal DynamicMaterialParam:Vector4f=&MakeVector4(X=1, Y=1, Z=&RandomRangeFloat(Minimum=0, Maximum=100), W=0)) {isSetParameters=true, module=SetVariables_#1}
   &AddVelocity(Velocity=&RandomRangeVector(Minimum=<0 0 30>, Maximum=<0 0 20>)) {enabled=false, module=AddVelocity001}
  ]
  particleUpdateScript [
   &ParticleState()
   /SetVariables(Particles.PartBlend:NiagaraFloat=&Sine(
     Normalized Angle=&Add_Float(A=&RandomRangeFloat(Minimum=0, Maximum=1), B=$Emitter.Age)
     Period=&RandomRangeFloat(Minimum=2.5, Maximum=5)
     Scale=0.5
     Bias=0.5
    )) {isSetParameters=true}
   &SkeletalMeshLocation(
    Bone / Socket Index=$Particles.RandomBone
    Bone Sampling Mode=::NewEnumerator4
    MaskSpawnGroup=false
    Sampled Position Offset=&self:Part_Selector(
     Bone Index:NiagaraInt32=$Particles.RandomBone
     Left Calve:Vector3f=&Lerp_Vector(A=<0 -1 -5>, B=<-30 -1 -5>, Alpha=$Particles.PartBlend)
     Left Shoe:Vector3f=&Lerp_Vector(A=<-8 -2 -3>, B=<-8 18 -5>, Alpha=$Particles.PartBlend)
     Left Shoulder:Vector3f=&Lerp_Vector(A=<7 -2 6>, B=<12 -2 6>, Alpha=$Particles.PartBlend)
     Left Thigh:Vector3f=&Lerp_Vector(A=<-10 1 -7>, B=<-35 1 -6>, Alpha=$Particles.PartBlend)
     Right Calve:Vector3f=&Lerp_Vector(A=<0 1 5>, B=<30 1 5>, Alpha=$Particles.PartBlend)
     Right Shoe:Vector3f=&Lerp_Vector(A=<8 -2 3>, B=<8 -18 5>, Alpha=$Particles.PartBlend)
     Right Shoulder:Vector3f=&Lerp_Vector(A=<-7 2 -6>, B=<-12 2 -6>, Alpha=$Particles.PartBlend)
     Right Thigh:Vector3f=&Lerp_Vector(A=<10 -1 7>, B=<35 -1 6>, Alpha=$Particles.PartBlend)
    )
    Skeletal Mesh=$User.SkeletalMesh
   ) {enabled=false}
   &SkeletalMeshLocation(Bone / Socket Index=$Particles.RandomBone, Bone Sampling Mode=::NewEnumerator4, MaskSpawnGroup=false, Skeletal Mesh=$User.SkeletalMeshRibbon) {module=SkeletalMeshLocation001}
   &SolveForcesAndVelocity(Clamp Velocity=false, Limit Acceleration=false)
   &ScaleColor(
    Scale Alpha=&FloatFromCurve(
     FloatCurve=DataInterface() {propertyValues {Curve {keys [
         {arriveTangent=9.0725, interpMode=RCIM_Cubic, leaveTangent=10.0491, tangentMode=RCTM_Break, time=0, value=0}
         {interpMode=RCIM_Cubic, tangentMode=RCTM_User, time=0.1, value=1}
         {interpMode=RCIM_Cubic, tangentMode=RCTM_User, time=0.5, value=1}
         {interpMode=RCIM_Cubic, leaveTangent=-0.0405, time=1, value=0}
        ]}}}
     Scale Curve=0.5
    )
    Scale RGB=&VectorFromFloat(Value=&FloatFromCurve(FloatCurve=DataInterface() {propertyValues {Curve {keys [
          {arriveTangent=9.0725, interpMode=RCIM_Cubic, time=0, value=0.1}
          {interpMode=RCIM_Cubic, tangentMode=RCTM_User, time=0.1, value=1}
          {interpMode=RCIM_Cubic, tangentMode=RCTM_User, time=0.5, value=1}
          {interpMode=RCIM_Cubic, leaveTangent=-0.0405, time=1, value=0.1}
         ]}}}))
    ScaleA=true
    ScaleRGB=true
   )
   &FN_Sprite_Attributes(
    Camera Offset=6
    Color=&MultiplyLinearColor(
     Linear Color=<rgba 0.25 1.8333 5 1>
     Scale Factor=&MakeVector4_FromFloat(Float=&FloatFromCurve(
       FloatCurve=DataInterface() {propertyValues {Curve {keys [
           {arriveTangent=9.0725, interpMode=RCIM_Cubic, leaveTangent=9.0725, tangentMode=RCTM_User, time=0, value=0}
           {arriveTangent=-2.5249, interpMode=RCIM_Cubic, leaveTangent=-2.5249, tangentMode=RCTM_User, time=0.1408, value=0.9847}
           {arriveTangent=-0.0405, interpMode=RCIM_Cubic, leaveTangent=-0.0405, tangentMode=RCTM_User, time=1, value=0}
          ]}}}
       Scale Curve=&RandomRangeFloat(Minimum=1.5, Maximum=2.5)
      ))
    )
    Create Camera Offset Attribute=true
    Create Color Attribute=true
    Write to Module Namespace=true
   )
   &FN_Sprite_Attributes(
    Camera Offset=20
    Color=&MultiplyLinearColor(
     Linear Color=<rgba 0.15 1.1 3 0.05>
     Scale Factor=&MakeVector4_FromFloat(Float=&FloatFromCurve(
       FloatCurve=DataInterface() {propertyValues {Curve {keys [
           {arriveTangent=9.0725, interpMode=RCIM_Cubic, leaveTangent=9.0725, tangentMode=RCTM_User, time=0, value=0}
           {arriveTangent=-2.5249, interpMode=RCIM_Cubic, leaveTangent=-2.5249, tangentMode=RCTM_User, time=0.1408, value=0.9847}
           {arriveTangent=-0.0405, interpMode=RCIM_Cubic, leaveTangent=-0.0405, tangentMode=RCTM_User, time=1, value=0}
          ]}}}
       Scale Curve=&RandomRangeFloat(Minimum=1, Maximum=1)
      ))
    )
    Create Camera Offset Attribute=true
    Create Color Attribute=true
    Create SpriteSize Attribute=true
    SpriteSize=&Multiply_Vector2D_ByFloat(Float=1, Vector2D=$Particles.SpriteSize)
    Write to Module Namespace=true
   ) {module=FN_Sprite_Attributes001}
  ]
  rendererClasses [&NiagaraSpriteRendererProperties]
  renderers [
   {rendererClass=&NiagaraSpriteRendererProperties}
   {rendererClass=&NiagaraSpriteRendererProperties, rendererIndex=1}
  ]
 }
}
summary {
 systemName=NS_Blue_Cheese_M_OnBody_Burst
 emitters [
  {emitterName=Arcs, rendererClasses [&NiagaraMeshRendererProperties, &NiagaraSpriteRendererProperties]}
  {emitterName=BoneFlow_ArcEnergy001, rendererClasses [&NiagaraRibbonRendererProperties]}
  {emitterName=Decal_Burst, rendererClasses [&NiagaraSpriteRendererProperties]}
 ]
 userVariables [
  User.FinalStage {type=NiagaraFloat, defaultValue=0, description=""}
  User.SkeletalMesh {
   type=NiagaraDataInterfaceSkeletalMesh
   defaultValue {
    struct=NiagaraExt_VariableValue_DataInterface
    value {dataInterface=&self:NiagaraDataInterfaceSkeletalMesh_0}
   }
  }
  User.SkeletalMeshRibbon {
   type=NiagaraDataInterfaceSkeletalMesh
   defaultValue {
    struct=NiagaraExt_VariableValue_DataInterface
    value {dataInterface=&self:NiagaraDataInterfaceSkeletalMesh_1}
   }
  }
  User.VariantNum {type=NiagaraInt32}
 ]
}
