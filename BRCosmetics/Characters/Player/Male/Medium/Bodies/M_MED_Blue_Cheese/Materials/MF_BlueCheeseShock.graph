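# Material function for the Blue Cheese character outfit that sets an animated electric shock and glow emissive on incoming material attributes.
# A packed body mask supplies shock, glow and flow channels, with flow values offsetting time-panned noise lookups for flicker and shifting.
# Sine-driven flow and two blended tilings of electric noise form a contrast-sharpened shock value.
# A curve atlas colors the shock, which joins the tinted glow under a global strength multiplier.
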MaterialFunction /BRCosmetics/Characters/Player/Male/Medium/Bodies/M_MED_Blue_Cheese/Materials/MF_BlueCheeseShock
graph {nodes [
  FunctionOutput(None=SetMaterialAttributes!(MaterialAttributes=FunctionInput() {inputType=FunctionInput_MaterialAttributes}, EmissiveColor=NamedRerouteUsage()) {attributeSetTypes [#1]})
  LinearInterpolate()
  n1 = TextureSample() {
   samplerType=SAMPLERTYPE_Masks
   texture=/BRCosmetics/Characters/Player/Male/Medium/Bodies/M_MED_Blue_Cheese/Textures/T_M_MED_Blue_Cheese_Shock_Body_MASK
  }
  n2 = ChannelMaskParameter(None=@n1.RGB) {group=BlueCheeseShock, maskChannel=Green, parameterName=ShockMaskChannel, defaultValue=<rgba 0 1 0 0>}
  NamedRerouteDeclaration(None=@n2.R) {name=ShockMask, nodeColor=<rgba 0 1 0.6953 1>}
  n3 = ChannelMaskParameter(None=@n1.RGB) {group=BlueCheeseShock, maskChannel=Blue, parameterName=GlowMaskChannel, defaultValue=<rgba 0 0 1 0>}
  NamedRerouteDeclaration(None=@n3.R) {name=GlowMask, nodeColor=<rgba 1 0.3047 0 1>}
  NamedRerouteDeclaration(None=@n1.A) {name=ShockFlowMask, nodeColor=<rgba 1 0 0.3516 1>}
  n4 = Multiply(A=Time(), B=ScalarParameter() {defaultValue=1, group=BlueCheeseShock, parameterName=ShockNoiseFlickerSpeed})
  n5 = Multiply(A=NamedRerouteUsage(), B=ScalarParameter() {defaultValue=1, group=BlueCheeseShock, parameterName=ShockFlowMaskNoiseStrength})
  NamedRerouteDeclaration(None=TextureSample(UVs=Add(A=@n4, B=@n5)) {samplerType=SAMPLERTYPE_LinearColor, texture=/Game/Effects/Fort_Effects/Textures/Tile/LinerNoise64x64}.R) {name=ShockFlicker, nodeColor=<rgba 0.7578 1 0 1>}
  n6 = TextureCoordinate()
  n7 = Multiply(A=@n6, B=ScalarParameter() {defaultValue=5, group=BlueCheeseShock, parameterName=ShockNoiseTiling_1})
  n8 = TextureObjectParameter() {
   group=BlueCheeseShock
   parameterName=ShockNoise
   samplerType=SAMPLERTYPE_LinearColor
   texture=/Game/Characters/Player/Male/Medium/Bodies/M_MED_SRDrift/Textures/T_ElectricNoise_Mask
  }
  n9 = Add(A=Multiply(A=@n6, B=ScalarParameter() {defaultValue=10, group=BlueCheeseShock, parameterName=ShockNoiseTiling_2})) {constB=-0.5}
  NamedRerouteDeclaration(None=LinearInterpolate(
    A=TextureSample(UVs=@n7, Tex=@n8) {samplerType=SAMPLERTYPE_Masks, texture=/Game/Characters/Player/Male/Medium/Bodies/M_MED_Storm_Soldier/Textures/ElectricityNoise_M}.R
    B=TextureSample(UVs=@n9, Tex=@n8) {samplerType=SAMPLERTYPE_Masks, texture=&ElectricityNoise_M}.R
    Alpha=NamedRerouteUsage()
   )) {name=ShockNoise, nodeColor=<rgba 0.2188 1 0 1>}
  n10 = ScalarParameter() {defaultValue=0.15, group=BlueCheeseShock, parameterName=GlowMaskShockInfluence}
  n11 = NamedRerouteUsage()
  n12 = Multiply(A=Time(), B=ScalarParameter() {defaultValue=0.02, group=BlueCheeseShock, parameterName=ShockShiftSpeed})
  n13 = Multiply(A=NamedRerouteUsage(), B=ScalarParameter() {defaultValue=0.2, group=BlueCheeseShock, parameterName=ShockFlowMaskStrength})
  n14 = Subtract(
   A=TextureSample(UVs=Add(A=@n12, B=@n13)) {samplerType=SAMPLERTYPE_LinearColor, texture=&LinerNoise64x64}.G
   B=NamedRerouteUsage()
  ) {constA=0.28}
  n15 = Add(
   A=Multiply(A=Abs(None=@n14), B=ScalarParameter() {defaultValue=1, group=BlueCheeseShock, parameterName=ShockFrequency})
   B=Multiply(A=ScalarParameter() {defaultValue=-0.2, group=BlueCheeseShock, parameterName=ShockSpeed}, B=Time())
  )
  n16 = Saturate(None=ConstantBiasScale(None=Sine(None=@n15)))
  n17 = NamedRerouteUsage()
  NamedRerouteDeclaration(None=Add(A=Multiply(A=@n10, B=@n11), B=Multiply(A=@n16, B=@n17))) {name=ShockFlow, nodeColor=<rgba 0.5469 0 1 1>}
  n18 = Add(A=Subtract(A=NamedRerouteUsage(), B=Constant() {r=1}), B=Multiply(A=NamedRerouteUsage(), B=Constant() {r=2}))
  NamedRerouteDeclaration(None=MaterialFunctionCall(In=Saturate(None=@n18), Contrast=ScalarParameter() {defaultValue=0.1, group=BlueCheeseShock, parameterName=ShockNoiseContrast}) {materialFunction=&CheapContrast}.Result) {name=Shock, nodeColor=<rgba 0.6172 1 0 1>}
  n19 = Multiply(
   A=VectorParameter() {group=BlueCheeseShock, parameterName=GlowColor, defaultValue=<rgba 0 0.0253 1 1>}.RGB
   B=Multiply(A=ScalarParameter() {defaultValue=0.25, group=BlueCheeseShock, parameterName=GlowStrength}, B=NamedRerouteUsage())
  )
  n20 = CurveAtlasRowParameter(CurveTime=NamedRerouteUsage()) {
   atlas=/BRCosmetics/Characters/Player/Male/Medium/Bodies/M_MED_Blue_Cheese/Textures/C_M_MED_Blue_Cheese_Shock
   curve=/BRCosmetics/Characters/Player/Male/Medium/Bodies/M_MED_Blue_Cheese/Textures/C_MED_Blue_Cheese_Shock_01
   group=BlueCheeseShock
   parameterName=ShockEmissiveCurve
  }
  n21 = Multiply(A=@n20, B=ScalarParameter() {defaultValue=1, group=BlueCheeseShock, parameterName=ShockEmissiveStrength})
  NamedRerouteDeclaration(None=Multiply(A=ScalarParameter() {group=BlueCheeseShock, parameterName=GlobalEmissiveStrength}, B=Add(A=@n19, B=@n21))) {name=EmissiveOut, nodeColor=<rgba 0 0.4141 1 1>}
 ]}
