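# Character body material for the Blue Cheese Shock variant of the male medium player cosmetic, built on material attributes with Default Lit and Subsurface Profile shading.
# The shared character PreFX function feeds static-switched skin FX layers in series: silk, custom crest, metal LUT, then the Blue Cheese Shock effect, which defaults on.
# A shared SkinFX mask texture drives the silk and metal layers.
# The result passes through character PostFX and cinematic PostFX functions.
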Material /BRCosmetics/Characters/Player/Male/Medium/Bodies/M_MED_Blue_Cheese/Materials/M_MED_Blue_Cheese_Shock
graph {
 properties {
  bUseMaterialAttributes=true
  bUsedWithMeshDeformer=true
  bUsedWithMorphTargets=true
  bUsedWithSkeletalMesh=true
  pixelDepthOffsetMode=PDOM_Legacy
  shadingModel=MSM_FromMaterialExpression
  usedShadingModels=Default Lit | Subsurface Profile
 }
 nodes [
  n1 = MaterialFunctionCall() {materialFunction=&BaseFN_CharacterFunction_PreFX}
  n2 = TextureSample(Tex=TextureObjectParameter() {group=Skin FX, parameterName=SkinFX_Mask})
  n3 = StaticSwitchParameter(
   True=MaterialFunctionCall!(Alpha=ChannelMaskParameter(None=@n2.RGBA) {group=Skin FX, parameterName=Silk Channel}.R, Material=@n1.Result) {materialFunction=/Game/Weapons/WeaponSkins/MaterialFunction/MF_Silk}.Result
   False=@n1.Result
  ) {group=Skin FX, parameterName=Use Silk}
  n4 = StaticSwitchParameter(True=MaterialFunctionCall!(Materials=@n3) {materialFunction=/Game/Weapons/WeaponSkins/MaterialFunction/CustomCrests}.Result, False=@n3) {group=Skin FX, parameterName=Use Custom Crest}
  n5 = StaticSwitchParameter(
   True=MaterialFunctionCall(Mask=@n2.RGBA, Materials=@n4) {materialFunction=&MF_MetalsLUT}.Result
   False=@n4
  ) {group=Skin FX, parameterName=UseMetalLUT}
  n6 = MaterialFunctionCall(
   AO=@n1.AO
   Distance Blend=@n1.AO
   Fresnel=@n1.AO
   Gmap=@n1.AO
   Pre Skinned Local Position=@n1.AO
   Material=StaticSwitchParameter(
    True=MaterialFunctionCall!(In=@n5) {materialFunction=&/MF_BlueCheeseShock}.Result
    False=@n5
   ) {defaultValue=true, group=Skin FX, parameterName=UseBlueCheeseShock}
  ) {materialFunction=&BaseFN_CharacterFunction_PostFX}
  MaterialFunctionCall(In=@n6.Result) {materialFunction=&CINE_MF_CharPostFX}
 ]
}
