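# Post-animation corrective Control Rig for the M_MED_Blue_Cheese player body, fixing jacket, zipper and foot deformation.
# It runs the shared DeformRig function, constrains the zipper to weighted bones, then measures pelvis height, spine bend and thigh spherical pose readers.
# These values, some Kalman-smoothed, weight additive offsets on jacket master bones, with an extra emote-only branch.
# Foot yaw drives foot correctives; a pelvis-to-jacket constraint chain is left unconnected from execution.
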ControlRigBlueprint /BRCosmetics/Characters/Player/Male/Medium/Bodies/M_MED_Blue_Cheese/Meshes/M_MED_Blue_Cheese_CtrlRig
graphs [
 RigVMFunctionLibrary {links [], nodes []}
 RigVMModel {nodes [
   n1 = RigVMVariableNode() {variable=Relativ_Pevlis_Root}
   &RigUnit_ParentConstraint(
    AdvancedSettings=(InterpolationType=Average,RotationOrderForFilter=XZY)
    Child=(Type=Bone,Name="dyn_jacket_fr_l_01")
    ExecutePin=()
    Parents=((Item=(Type=Bone,Name="pelvis"),Weight=1))
    Weight=@n1.Value
   )
   > &RigUnit_ParentConstraint(
    AdvancedSettings=(InterpolationType=Average,RotationOrderForFilter=XZY)
    Child=(Type=Bone,Name="dyn_jacket_l_master")
    Parents=((Item=(Type=Bone,Name="pelvis"),Weight=1))
    Weight=@n1.Value
   )
   > &RigUnit_ParentConstraint(
    AdvancedSettings=(InterpolationType=Average,RotationOrderForFilter=XZY)
    Child=(Type=Bone,Name="dyn_jacket_fr_r_01")
    Parents=((Item=(Type=Bone,Name="pelvis"),Weight=1))
    Weight=@n1.Value
   )
   > &RigUnit_ParentConstraint(
    AdvancedSettings=(InterpolationType=Average,RotationOrderForFilter=XZY)
    Child=(Type=Bone,Name="dyn_jacket_r_01")
    Parents=((Item=(Type=Bone,Name="pelvis"),Weight=1))
    Weight=@n1.Value
   )
   RigUnit_BeginExecution = &RigUnit_BeginExecution()
   FN_DeformRig_FNC = RigVMFunctionReferenceNode(
    DeformRig_ Curve="(deformRig_l_upperarm_up_alpha_72_#1=False,deformRig_r_upperarm_up_alpha_73_#2=False,deformRig_l_upperarm_fr_alpha_74_#3=False,deformRig_r_upperarm_fr_alpha_75_CB51B8 ...(truncated)"
    DeformRig_Bool="(upperarm_1_#4=True,elbow_3_#5=True,wrist_7_#6=True,groin_8_#7=True,glute_17_#8=True,knee_9_40C408324CA261 ...(truncated)"
    DeformRig_Debug="(upperarm_debug=False,wrist_debug=False,lat_debug=False,groinGlute_debug=False,upperarm_debug_18_#4=False,wrist_debug_20_#6=False,groinGlute_debug_27_#7=False,lat_deb ...(truncated)"
    DeformRig_Value="(upperarm_up_18_#9=(Translation=(X=0,Y=0,Z=8),Scale3D=(X=0.7,Y=1,Z=1)),upperarm_fwd_fr_19_#10=(Translation=(X=0,Y=10,Z=0),Scale3D=(X=0.5,Y=1,Z=1)),upperarm_fwd_bk_20_9973AAD84EE0A1E349BC9E85067D4A3 ...(truncated)"
    ExecuteContext=@RigUnit_BeginExecution.ExecutePin
   )
   > &RigUnit_ParentConstraint(
    AdvancedSettings=(InterpolationType=Average,RotationOrderForFilter=XZY)
    Child=(Type=Bone,Name="dyn_zipper_master")
    Parents="((Item=(Type=Bone,Name=\"spine_03\"),Weight=0.0023),(Item=(Type=Bone,Name=\"dyn_jacket_fr_r_01\"),Weight=0.0315),(Item=(Type=Bone,Name=\"dyn_jacket_fr_r_02\"),Weight=0.8842),(Item=(Type=Bone,Name=\"dyn_jacket_r_01\"),Weight=0.0016),(Item=(Type=Bone,Name=\"dyn_jacke ...(truncated)"
    Weight=1
   )
   > Get Delta Ref = RigVMFunctionReferenceNode(Item=(Type=Bone,Name="calf_l"))
   > RigVMVariableNode(Value=@"Get Delta Ref".Result.Rotation.Yaw) {variable=Calf_L_Delta}
   > Get Delta Ref_1 = RigVMFunctionReferenceNode(Item=(Type=Bone,Name="calf_r"))
   > RigVMVariableNode(Value=@"Get Delta Ref_1".Result.Rotation.Yaw) {variable=Calf_R_Delta}
   RigVMFunction_MathQuaternionToEuler = &RigVMFunction_MathQuaternionToEuler(
    RotationOrder=ZYX
    Value=&RigUnit_GetRelativeTransformForItem(Child=(Type=Bone,Name="pelvis"), Parent=(Type=Bone,Name="spine_05"), bChildInitial=False, bParentInitial=False).RelativeTransform.Rotation
   )
   > RigVMVariableNode(Value=@RigVMFunction_MathQuaternionToEuler.Result.X) {variable=Relavtive spine X}
   > RigVMVariableNode(Value=@RigVMFunction_MathQuaternionToEuler.Result.Z) {variable=Relavtive spine z}
   n2 = &RigUnit_GetRelativeTransformForItem(Child=(Type=Bone,Name="pelvis"), Parent=(Type=Bone,Name="root"), bChildInitial=False, bParentInitial=False)
   > RigVMVariableNode(Value=&RigVMFunction_MathDoubleRemap(SourceMaximum=80, SourceMinimum=74, TargetMaximum=0, TargetMinimum=1, Value=@n2.RelativeTransform.Translation.Z).Result) {variable=Relativ_Pevlis_Root}
   Remap = &RigVMFunction_MathDoubleRemap(SourceMaximum=44, SourceMinimum=60, TargetMaximum=0, TargetMinimum=1, Value=@n2.RelativeTransform.Translation.Z)
   > RigVMVariableNode(Value=@Remap.Result) {variable=Relativ_Pelvis_Root_Walk}
   > n3 = &RigUnit_SphericalPoseReader(
    ActiveRegionScaleFactors=(PositiveWidth=0.1,NegativeWidth=1,PositiveHeight=1,NegativeHeight=1)
    ActiveRegionSize=0.25
    Debug=(bDrawDebug=false,bDraw2D=False,bDrawLocalAxes=False,DebugScale=25,DebugSegments=20,DebugThickness=0.25)
    DriverAxis=(X=1,Y=0,Z=0)
    DriverItem=(Type=Bone,Name="thigh_r")
    FalloffRegionScaleFactors=(PositiveWidth=0.9,NegativeWidth=1,PositiveHeight=1,NegativeHeight=1)
    FalloffSize=0.4
    FlipHeightScaling=False
    FlipWidthScaling=False
    OptionalParentItem=(Type=Bone,Name="None")
    RotationOffset=(X=-90,Y=0,Z=0)
   )
   > RigVMVariableNode(Value=@n3.OutputParam) {variable=SPR_Thigh_r}
   > n4 = &RigUnit_SphericalPoseReader(
    ActiveRegionScaleFactors=(PositiveWidth=1,NegativeWidth=0.1,PositiveHeight=1,NegativeHeight=1)
    ActiveRegionSize=0.25
    Debug=(bDrawDebug=false,bDraw2D=False,bDrawLocalAxes=False,DebugScale=25,DebugSegments=20,DebugThickness=0.25)
    DriverAxis=(X=-1,Y=0,Z=0)
    DriverItem=(Type=Bone,Name="thigh_l")
    FalloffRegionScaleFactors=(PositiveWidth=1,NegativeWidth=0.9,PositiveHeight=1,NegativeHeight=1)
    FalloffSize=0.4
    FlipHeightScaling=False
    FlipWidthScaling=False
    OptionalParentItem=(Type=Bone,Name="None")
    RotationOffset=(X=90,Y=0,Z=0)
   )
   > RigVMVariableNode(Value=@n4.OutputParam) {variable=SPR_Thig_l}
   > n5 = &RigUnit_SphericalPoseReader(
    ActiveRegionScaleFactors=(PositiveWidth=0.1,NegativeWidth=1,PositiveHeight=0.4,NegativeHeight=0.3)
    ActiveRegionSize=0.25
    Debug=(bDrawDebug=true,bDraw2D=False,bDrawLocalAxes=False,DebugScale=25,DebugSegments=20,DebugThickness=0.25)
    DriverAxis=(X=1,Y=0,Z=0)
    DriverItem=(Type=Bone,Name="thigh_r")
    FalloffRegionScaleFactors=(PositiveWidth=0.9,NegativeWidth=1,PositiveHeight=0.4,NegativeHeight=0.3)
    FalloffSize=0.4
    FlipHeightScaling=False
    FlipWidthScaling=False
    OptionalParentItem=(Type=Bone,Name="None")
    RotationOffset=(X=0,Y=0,Z=0)
   )
   > RigVMVariableNode(Value=@n5.OutputParam) {variable=SPR_Thigh_Side_r}
   > n6 = &RigUnit_SphericalPoseReader(
    ActiveRegionScaleFactors=(PositiveWidth=1,NegativeWidth=0.1,PositiveHeight=0.4,NegativeHeight=0.4)
    ActiveRegionSize=0.25
    Debug=(bDrawDebug=true,bDraw2D=False,bDrawLocalAxes=False,DebugScale=25,DebugSegments=20,DebugThickness=0.25)
    DriverAxis=(X=-1,Y=0,Z=0)
    DriverItem=(Type=Bone,Name="thigh_l")
    FalloffRegionScaleFactors=(PositiveWidth=1,NegativeWidth=0.9,PositiveHeight=0.4,NegativeHeight=0.4)
    FalloffSize=0.4
    FlipHeightScaling=False
    FlipWidthScaling=False
    OptionalParentItem=(Type=Bone,Name="None")
    RotationOffset=(X=180,Y=0,Z=0)
   )
   > RigVMVariableNode(Value=@n6.OutputParam) {variable=SPR_Thig_Side_l}
   KalmanFilter = &RigVMFunction_KalmanFloat(BufferSize=16, Value=RigVMVariableNode() {variable=SPR_Thig_l}.Value) {template="KalmanFilter::Execute(in Value,in BufferSize,out Result)"}
   > &RigUnit_ModifyTransforms(
    ItemToModify="((Item=(Type=Bone,Name=\"dyn_jacket_fr_l_master\"),Transform=(Rotation=(X=0,Y=-0.7373,Z=0,W=0.6756),Translation=(X=0,Y=0,Z=0),Scale3D=(X=1,Y=1,Z=1))),(Item=(Type=Bone,Name=\"dyn_jacket_l_master\"),Transform=(Rotation=(X=0,Y=-0.342,Z=0,W=0.9397),Translation=(X= ...(truncated)"
    WeightMinimum=0
    Weight=@KalmanFilter.Result
   )
   > &RigUnit_ModifyTransforms(
    ItemToModify="((Item=(Type=Bone,Name=\"dyn_jacket_fr_r_master\"),Transform=(Rotation=(X=0,Y=-0.5736,Z=0,W=0.8192),Translation=(X=0,Y=0,Z=0),Scale3D=(X=1,Y=1,Z=1))),(Item=(Type=Bone,Name=\"dyn_jacket_r_master\"),Transform=(Rotation=(X=-0.1094,Y=-0.2346,Z=-0.4082,W=0.8754),Tr ...(truncated)"
    WeightMinimum=0
    Weight=&RigVMFunction_KalmanFloat(BufferSize=16, Value=RigVMVariableNode() {variable=SPR_Thigh_r}.Value) {template="KalmanFilter::Execute(in Value,in BufferSize,out Result)"}.Result
   )
   > &RigUnit_ModifyTransforms(
    ItemToModify="((Item=(Type=Bone,Name=\"dyn_jacket_bk_l_master\"),Transform=(Rotation=(X=0,Y=0,Z=0,W=1),Translation=(X=0,Y=0,Z=3),Scale3D=(X=1,Y=1,Z=1))),(Item=(Type=Bone,Name=\"dyn_jacket_bk_r_master\"),Transform=(Rotation=(X=0,Y=0,Z=0,W=1),Translation=(X=0,Y=0,Z=3),Scale3D ...(truncated)"
    Weight=1
    WeightMinimum=0
   )
   Multiply = &RigVMFunction_MathDoubleMul(B=RigVMVariableNode() {variable=Relativ_Pelvis_Root_Walk}.Value, A=RigVMVariableNode() {variable=Relativ_Pevlis_Root}.Value)
   > &RigUnit_ModifyTransforms(
    ItemToModify="((Item=(Type=Bone,Name=\"dyn_jacket_bk_l_master\"),Transform=(Rotation=(X=0,Y=0,Z=0,W=1),Translation=(X=3,Y=0,Z=0),Scale3D=(X=1,Y=1,Z=1))),(Item=(Type=Bone,Name=\"dyn_jacket_bk_r_master\"),Transform=(Rotation=(X=0,Y=0,Z=0,W=1),Translation=(X=4,Y=0,Z=0),Scale3D ...(truncated)"
    WeightMinimum=0
    Weight=&RigVMFunction_KalmanFloat(BufferSize=4, Value=@Multiply.Result) {template="KalmanFilter::Execute(in Value,in BufferSize,out Result)"}.Result
   )
   > &RigUnit_ModifyTransforms(
    ItemToModify="((Item=(Type=Bone,Name=\"dyn_jacket_bk_l_master\"),Transform=(Rotation=(X=0,Y=-0.1736,Z=0,W=0.9848),Translation=(X=0,Y=0,Z=0),Scale3D=(X=1,Y=1,Z=1))),(Item=(Type=Bone,Name=\"dyn_jacket_bk_r_master\"),Transform=(Rotation=(X=0,Y=-0.1736,Z=0,W=0.9848),Translation ...(truncated)"
    WeightMinimum=0
    Weight=&RigVMFunction_MathDoubleRemap(SourceMaximum=100, SourceMinimum=80, TargetMaximum=1, TargetMinimum=0, Value=RigVMVariableNode() {variable=Relavtive spine z}.Value).Result
   )
   > Get Delta Ref_2 = RigVMFunctionReferenceNode(Item=(Type=Bone,Name="foot_l"))
   > &RigUnit_ModifyTransforms(
    ItemToModify=((Item=(Type=Bone,Name="Corrective_foot_l"),Transform=(Rotation=(X=0,Y=0,Z=0,W=1),Translation=(X=3,Y=-3,Z=0),Scale3D=(X=1,Y=1,Z=1))))
    WeightMinimum=0
    Weight=&RigVMFunction_MathDoubleRemap(SourceMaximum=45, SourceMinimum=35, TargetMaximum=1, TargetMinimum=0, Value=@"Get Delta Ref_2".Result.Rotation.Yaw).Result
   )
   > Get Delta Ref_3 = RigVMFunctionReferenceNode(Item=(Type=Bone,Name="foot_r"))
   > &RigUnit_ModifyTransforms(
    ItemToModify=((Item=(Type=Bone,Name="Corrective_foot_r"),Transform=(Rotation=(X=0,Y=0,Z=0,W=1),Translation=(X=-3,Y=-3,Z=0),Scale3D=(X=1,Y=1,Z=1))))
    WeightMinimum=0
    Weight=&RigVMFunction_MathDoubleRemap(SourceMaximum=45, SourceMinimum=35, TargetMaximum=1, TargetMinimum=0, Value=@"Get Delta Ref_3".Result.Rotation.Yaw).Result
   )
   n7 = RigVMVariableNode() {variable=Relavtive spine X}
   n8 = RigVMVariableNode() {variable=Relativ_Pevlis_Root}
   > &RigUnit_ModifyTransforms(
    ItemToModify=((Item=(Type=Bone,Name="dyn_jacket_fr_r_master"),Transform=(Rotation=(X=0,Y=0,Z=0,W=1),Translation=(X=0,Y=5,Z=0),Scale3D=(X=1,Y=1,Z=1))))
    WeightMinimum=0
    Weight=&RigVMFunction_MathDoubleMul(
     A=&RigVMFunction_MathDoubleRemap(SourceMaximum=-40, SourceMinimum=-20, TargetMaximum=1, TargetMinimum=0, Value=@n7.Value).Result
     B=@n8.Value
    ).Result
   )
   > n9 = &RigUnit_ModifyTransforms(
    ItemToModify=((Item=(Type=Bone,Name="dyn_jacket_fr_l_master"),Transform=(Rotation=(X=0,Y=0,Z=0,W=1),Translation=(X=5,Y=0,Z=0),Scale3D=(X=1,Y=1,Z=1))))
    WeightMinimum=0
    Weight=&RigVMFunction_MathDoubleMul(
     A=&RigVMFunction_MathDoubleRemap(SourceMaximum=40, SourceMinimum=30, TargetMaximum=1, TargetMinimum=0, Value=@n7.Value).Result
     B=@n8.Value
    ).Result
   )
   RigVMFunction_Sequence = &RigVMFunction_Sequence(ExecuteContext=@n9.ExecutePin)
   RigVMFunction_ControlFlowBranch = &RigVMFunction_ControlFlowBranch(ExecuteContext=@RigVMFunction_Sequence.A, Condition=RigVMVariableNode() {variable=Is play emote}.Value)
   &RigUnit_ModifyTransforms(
    ItemToModify="((Item=(Type=Bone,Name=\"dyn_jacket_master\"),Transform=(Rotation=(X=0,Y=0,Z=0,W=1),Translation=(X=0,Y=0,Z=0),Scale3D=(X=1,Y=1,Z=1))),(Item=(Type=Bone,Name=\"dyn_jacket_fr_l_master\"),Transform=(Rotation=(X=0,Y=0,Z=0,W=1),Translation=(X=0,Y=0,Z=4),Scale3D=(X=1 ...(truncated)"
    WeightMinimum=0
    Weight=&RigVMFunction_MathDoubleRemap(SourceMaximum=-24, SourceMinimum=-15, TargetMaximum=1, TargetMinimum=0, Value=RigVMVariableNode() {variable=Relavtive spine X}.Value).Result
    ExecutePin=@RigVMFunction_ControlFlowBranch.True
   )
   &RigUnit_ModifyTransforms(
    ItemToModify="((Item=(Type=Bone,Name=\"dyn_jacket_r_master\"),Transform=(Rotation=(X=-0.2988,Y=-0.6409,Z=-0.2988,W=0.6409),Translation=(X=0,Y=0,Z=0),Scale3D=(X=1,Y=1,Z=1))),(Item=(Type=Bone,Name=\"dyn_jacket_fr_r_master\"),Transform=(Rotation=(X=0,Y=-0.7071,Z=0,W=0.7071),Tr ...(truncated)"
    WeightMinimum=0
    ExecutePin=@RigVMFunction_Sequence.B
    Weight=RigVMVariableNode() {variable=SPR_Thigh_Side_r}.Value
   )
   > &RigUnit_ModifyTransforms(
    ItemToModify="((Item=(Type=Bone,Name=\"dyn_jacket_l_master\"),Transform=(Rotation=(X=0,Y=-0.7071,Z=0,W=0.7071),Translation=(X=0,Y=0,Z=0),Scale3D=(X=1,Y=1,Z=1))),(Item=(Type=Bone,Name=\"dyn_jacket_fr_l_master\"),Transform=(Rotation=(X=0,Y=-0.7071,Z=0,W=0.7071),Translation=(X ...(truncated)"
    WeightMinimum=0
    Weight=RigVMVariableNode() {variable=SPR_Thig_Side_l}.Value
   )
  ]}
]
hierarchy [
 root {loc=<0 0 0>, rot=<0 0 0 1>}
 pelvis {parent=root, loc=<0 -0.7071 93.1191>, rot=<0.0127 0.707 -0.0127 -0.707>}
 spine_01 {parent=pelvis, loc=<6.7494 -0.6388 0>, rot=<0 0 0.034 -0.9994>}
 spine_02 {parent=spine_01, loc=<7.2518 -0.035 0>, rot=<0 0 0.0238 -0.9997>}
 spine_03 {parent=spine_02, loc=<6.0104 0 0>, rot=<0 0 0.0149 -0.9999>}
 spine_04 {parent=spine_03, loc=<8.1024 0 0>, rot=<0 0 -0.0132 -0.9999>}
 spine_05 {parent=spine_04, loc=<10.0146 0 0>, rot=<0 0 -0.0015 -1>}
 clavicle_l {parent=spine_05, loc=<4.3575 -0.1466 -2.3223>, rot=<0.0465 -0.7289 -0.009 -0.6829>}
 upperarm_l {parent=clavicle_l, loc=<15.5111 0 0>, rot=<0.0226 -0.3234 -0.0088 -0.9459>}
 lowerarm_l {parent=upperarm_l, loc=<26.6514 0 0>, rot=<0.0391 -0.0099 -0.2453 -0.9686>}
 hand_l {parent=lowerarm_l, loc=<22.3491 0 0>, rot=<-0.7081 -0.0507 0.0975 -0.6975>}
 index_metacarpal_l {parent=hand_l, loc=<1.6535 0.0183 -2.96>, rot=<0.0158 -0.0124 -0.0173 -0.9996>}
 index_01_l {parent=index_metacarpal_l, loc=<8.3008 0.0357 -0.0524>, rot=<-0.0044 0.0485 0.0924 -0.9945>}
 index_02_l {parent=index_01_l, loc=<3.3587 -0.3004 0.0452>, rot=<-0.002 0.0027 0.3225 -0.9466>}
 index_03_l {parent=index_02_l, loc=<2.5578 -0.123 -0.0252>, rot=<-0.002 0.0067 0.2558 -0.9667>}
 middle_metacarpal_l {parent=hand_l, loc=<2.1856 -0.1725 -0.5984>, rot=<0.006 0.0168 -0.0329 -0.9993>}
 middle_01_l {parent=middle_metacarpal_l, loc=<7.3275 -0.2447 0.0129>, rot=<-0.0208 0.0042 0.0858 -0.9961>}
 middle_02_l {parent=middle_01_l, loc=<4.0123 0.1401 0.1175>, rot=<-0.0218 0.0258 0.3512 -0.9357>}
 middle_03_l {parent=middle_02_l, loc=<2.6077 0.0534 0.1373>, rot=<0.0082 0.0188 0.1708 -0.9851>}
 pinky_metacarpal_l {parent=hand_l, loc=<1.7616 -0.9753 3.3814>, rot=<-0.0397 0.0814 -0.069 -0.9935>}
 pinky_01_l {parent=pinky_metacarpal_l, loc=<7.1709 -0.4831 -0.1094>, rot=<-0.0654 -0.0035 0.2242 -0.9723>}
 pinky_02_l {parent=pinky_01_l, loc=<2.8274 0.0893 -0.1043>, rot=<0.0037 0.0011 0.2854 -0.9584>}
 pinky_03_l {parent=pinky_02_l, loc=<2.2046 0.0307 -0.0078>, rot=<-0.0002 0.0098 0.1445 -0.9895>}
 ring_metacarpal_l {parent=hand_l, loc=<2.2074 -0.3625 1.3638>, rot=<-0.0284 0.0722 -0.0458 -0.9959>}
 ring_01_l {parent=ring_metacarpal_l, loc=<7.0295 -0.6197 -0.1446>, rot=<-0.0489 -0.0252 0.1892 -0.9804>}
 ring_02_l {parent=ring_01_l, loc=<3.6174 0.4647 -0.1027>, rot=<0 0 0.2806 -0.9598>}
 ring_03_l {parent=ring_02_l, loc=<2.4152 0.114 0.0199>, rot=<0 0 0.1189 -0.9929>}
 thumb_01_l {parent=hand_l, loc=<1.3118 -0.3191 -2.459>, rot=<0.6546 -0.4042 0.0664 -0.6354>}
 thumb_02_l {parent=thumb_01_l, loc=<5.9336 -0.0448 0.0401>, rot=<-0.0037 0.0213 0.298 -0.9543>}
 thumb_03_l {parent=thumb_02_l, loc=<2.2223 0.3682 -0.0907>, rot=<0.0118 -0.013 0.0066 -0.9998>}
 deform_wrist_dn_base_l {parent=hand_l, loc=<0 -2.2342 0>, rot=<0 0 0 -1>}
 deform_wrist_dn_l {parent=deform_wrist_dn_base_l, loc=<0 0 0>, rot=<0 0 0 1>}
 deform_wrist_up_base_l {parent=hand_l, loc=<0 2.2682 0>, rot=<0 0 0 -1>}
 deform_wrist_up_l {parent=deform_wrist_up_base_l, loc=<0 0 0>, rot=<0 0 0 1>}
 weapon_l {parent=hand_l, loc=<-0.1818 -3.3618 0.3735>, rot=<-0.0012 0.0004 0.0004 -1>}
 lowerarm_twist_01_l {parent=lowerarm_l, loc=<17.8549 0 0>, rot=<0 0 0 -1>}
 lowerarm_twist_02_l {parent=lowerarm_l, loc=<9.6199 0 0>, rot=<0 0 0 -1>}
 deform_elbow_in_base_l {parent=lowerarm_l, loc=<0.2112 5.2591 0.8274>, rot=<0 0 0 1>}
 deform_elbow_in_l {parent=deform_elbow_in_base_l, loc=<0 0 0>, rot=<0 0 0 1>}
 deform_elbow_out_base_l {parent=lowerarm_l, loc=<0 -4 0>, rot=<0 0 0 1>}
 deform_elbow_out_l {parent=deform_elbow_out_base_l, loc=<0 0 0>, rot=<0 0 0 1>}
 upperarm_twist_01_l {parent=upperarm_l, loc=<0.0041 0 0>, rot=<0 0 0 -1>}
 deform_deltoid_fr_l {parent=upperarm_twist_01_l, loc=<2.4511 0 0>, rot=<0 0 0 -1>}
 deform_upperarm_bk_base_l {parent=upperarm_twist_01_l, loc=<1.0813 -7.2193 -1.9453>, rot=<0 0 0 1>}
 deform_upperarm_bk_l {parent=deform_upperarm_bk_base_l, loc=<0 0 0>, rot=<0 0 0 1>}
 deform_upperarm_fr_base_l {parent=upperarm_twist_01_l, loc=<-0.3691 7.0506 -2.765>, rot=<0 0 0 1>}
 deform_upperarm_fr_l {parent=deform_upperarm_fr_base_l, loc=<0 0 0>, rot=<0 0 0 1>}
 deform_upperarm_up_base_l {parent=upperarm_twist_01_l, loc=<-3.2216 -1.1391 3.3311>, rot=<0 0 0 1>}
 deform_upperarm_up_l {parent=deform_upperarm_up_base_l, loc=<0 0 0>, rot=<0 0 0 1>}
 upperarm_twist_02_l {parent=upperarm_l, loc=<13.7941 0 0>, rot=<0 0 0 -1>}
 dyn_l_sleeve_master {parent=upperarm_twist_02_l, loc=<0 0 0>, rot=<0 0 0 1>}
 dyn_l_sleeve {parent=dyn_l_sleeve_master, loc=<0 0 0>, rot=<0 0 0 1>}
 dyn_l_sleeve_back_master {parent=upperarm_l, loc=<13.2653 -6.2493 -1.3409>, rot=<-0.7191 0.0777 -0.0352 -0.6896>}
 dyn_l_sleeve_back {parent=dyn_l_sleeve_back_master, loc=<0 0 0>, rot=<0 0 0 1>}
 dyn_l_sleeve_front_master {parent=upperarm_l, loc=<13.5533 5.8403 -1.0283>, rot=<0.6993 0.1418 0.0394 -0.6996>}
 dyn_l_sleeve_front {parent=dyn_l_sleeve_front_master, loc=<0 0 0>, rot=<0 0 0 -1>}
 dyn_l_sleeve_top_master {parent=upperarm_l, loc=<12.878 -0.2476 3.6479>, rot=<0.01 0.1275 0.0105 -0.9917>}
 dyn_l_sleeve_top {parent=dyn_l_sleeve_top_master, loc=<0 0 0>, rot=<0 0 0 -1>}
 dyn_l_sleeve_bot_master {parent=upperarm_l, loc=<13.9312 -0.0799 -5.0025>, rot=<0.9994 0.0186 -0.0263 0.0154>}
 dyn_l_sleeve_bot {parent=dyn_l_sleeve_bot_master, loc=<0 0 0>, rot=<0 0 0 -1>}
 FX_Shoulder_L1 {parent=upperarm_l, type=Null, loc=<15.3817 -0.0003 3.9903>, rot=<0.0436 0 -0.999 0.0001>}
 deform_pectoral_fr_l {parent=clavicle_l, loc=<9.3905 0 0>, rot=<0 0 0 1>}
 FX_Shoulder_L2 {parent=clavicle_l, type=Null, loc=<10.4113 -1.8216 5.5566>, rot=<0 0.3007 0 0.9537>}
 clavicle_r {parent=spine_05, loc=<4.3575 -0.1466 2.3223>, rot=<0.7289 0.0465 -0.6829 0.009>}
 upperarm_r {parent=clavicle_r, loc=<-15.5111 0 0>, rot=<0.0226 -0.3234 -0.0088 -0.9459>}
 lowerarm_r {parent=upperarm_r, loc=<-26.6515 0 0>, rot=<0.0391 -0.0099 -0.2453 -0.9686>}
 hand_r {parent=lowerarm_r, loc=<-22.3492 0 0>, rot=<-0.7081 -0.0507 0.0975 -0.6975>}
 index_metacarpal_r {parent=hand_r, loc=<-1.6534 -0.0181 2.9601>, rot=<0.0158 -0.0124 -0.0173 -0.9996>}
 index_01_r {parent=index_metacarpal_r, loc=<-8.3008 -0.0357 0.0524>, rot=<-0.0044 0.0485 0.0924 -0.9945>}
 index_02_r {parent=index_01_r, loc=<-3.3589 0.3003 -0.0453>, rot=<-0.002 0.0027 0.3225 -0.9466>}
 index_03_r {parent=index_02_r, loc=<-2.5578 0.1231 0.0251>, rot=<-0.002 0.0067 0.2558 -0.9667>}
 middle_metacarpal_r {parent=hand_r, loc=<-2.1856 0.1726 0.5984>, rot=<0.006 0.0168 -0.0329 -0.9993>}
 middle_01_r {parent=middle_metacarpal_r, loc=<-7.3273 0.2447 -0.0129>, rot=<-0.0208 0.0042 0.0858 -0.9961>}
 middle_02_r {parent=middle_01_r, loc=<-4.0142 -0.1411 -0.07>, rot=<-0.0217 0.0259 0.3512 -0.9357>}
 middle_03_r {parent=middle_02_r, loc=<-2.6074 -0.0534 -0.1312>, rot=<0.0083 0.0188 0.1708 -0.9851>}
 pinky_metacarpal_r {parent=hand_r, loc=<-1.7617 0.9753 -3.3814>, rot=<-0.0397 0.0814 -0.069 -0.9935>}
 pinky_01_r {parent=pinky_metacarpal_r, loc=<-7.1707 0.4832 0.1094>, rot=<-0.0654 -0.0035 0.2242 -0.9723>}
 pinky_02_r {parent=pinky_01_r, loc=<-2.8274 -0.0892 0.1043>, rot=<0.0037 0.0011 0.2854 -0.9584>}
 pinky_03_r {parent=pinky_02_r, loc=<-2.2045 -0.031 0.0078>, rot=<-0.0002 0.0098 0.1445 -0.9895>}
 ring_metacarpal_r {parent=hand_r, loc=<-2.2073 0.3627 -1.3638>, rot=<-0.0284 0.0722 -0.0458 -0.9959>}
 ring_01_r {parent=ring_metacarpal_r, loc=<-7.0297 0.6197 0.1446>, rot=<-0.0489 -0.0252 0.1892 -0.9804>}
 ring_02_r {parent=ring_01_r, loc=<-3.6171 -0.4648 0.1027>, rot=<0 0 0.2806 -0.9598>}
 ring_03_r {parent=ring_02_r, loc=<-2.4153 -0.1141 -0.0199>, rot=<0 0 0.1189 -0.9929>}
 thumb_01_r {parent=hand_r, loc=<-1.3119 0.3192 2.459>, rot=<0.6546 -0.4042 0.0664 -0.6354>}
 thumb_02_r {parent=thumb_01_r, loc=<-5.9336 0.0447 -0.0401>, rot=<-0.0037 0.0213 0.298 -0.9543>}
 thumb_03_r {parent=thumb_02_r, loc=<-2.2223 -0.3682 0.0907>, rot=<0.0118 -0.013 0.0066 -0.9998>}
 deform_wrist_dn_base_r {parent=hand_r, loc=<-0.0003 2.2346 -0.0001>, rot=<0 0 0 -1>}
 deform_wrist_dn_r {parent=deform_wrist_dn_base_r, loc=<0 0 0>, rot=<0 0 0 1>}
 deform_wrist_up_base_r {parent=hand_r, loc=<0.0004 -2.2683 0.0001>, rot=<0 0 0 -1>}
 deform_wrist_up_r {parent=deform_wrist_up_base_r, loc=<0 0 0>, rot=<0 0 0 1>}
 weapon_r {parent=hand_r, loc=<-1.4133 2.4113 1.0144>, rot=<-0.303 0.0278 0.9433 -0.1324>}
 lowerarm_twist_01_r {parent=lowerarm_r, loc=<-17.855 0 0>, rot=<0 0 0 1>}
 lowerarm_twist_02_r {parent=lowerarm_r, loc=<-9.6199 0 0>, rot=<0 0 0 1>}
 deform_elbow_in_base_r {parent=lowerarm_r, loc=<-0.2111 -5.2591 -0.8275>, rot=<0 0 0 1>}
 deform_elbow_in_r {parent=deform_elbow_in_base_r, loc=<-0.0001 0 0.0002>, rot=<0 0 0 1>}
 deform_elbow_out_base_r {parent=lowerarm_r, loc=<0 4 0>, rot=<0 0 0 1>}
 deform_elbow_out_r {parent=deform_elbow_out_base_r, loc=<0 0 0>, rot=<0 0 0 1>}
 upperarm_twist_01_r {parent=upperarm_r, loc=<-0.0041 0 0>, rot=<0 0 0 1>}
 deform_deltoid_fr_r {parent=upperarm_twist_01_r, loc=<-2.4512 0 0>, rot=<0 0 0 1>}
 deform_upperarm_bk_base_r {parent=upperarm_twist_01_r, loc=<-1.0815 7.2192 1.9456>, rot=<0 0 0 1>}
 deform_upperarm_bk_r {parent=deform_upperarm_bk_base_r, loc=<0 0 0>, rot=<0 0 0 1>}
 deform_upperarm_fr_base_r {parent=upperarm_twist_01_r, loc=<0.3689 -7.0506 2.7652>, rot=<0 0 0 1>}
 deform_upperarm_fr_r {parent=deform_upperarm_fr_base_r, loc=<0 0 0>, rot=<0 0 0 1>}
 deform_upperarm_up_base_r {parent=upperarm_twist_01_r, loc=<3.2213 1.1391 -3.3308>, rot=<0 0 0 1>}
 deform_upperarm_up_r {parent=deform_upperarm_up_base_r, loc=<0 0 0>, rot=<0 0 0 -1>}
 upperarm_twist_02_r {parent=upperarm_r, loc=<-13.7942 0 0>, rot=<0 0 0 1>}
 dyn_r_sleeve_master {parent=upperarm_twist_02_r, loc=<0 0 0>, rot=<0 0 0 1>}
 dyn_r_sleeve {parent=dyn_r_sleeve_master, loc=<0 0 0>, rot=<0 0 0 1>}
 dyn_r_sleeve_back_master {parent=upperarm_r, loc=<-13.2863 6.1601 0.4695>, rot=<-0.0161 0.6803 0.7246 0.1093>}
 dyn_r_sleeve_back {parent=dyn_r_sleeve_back_master, loc=<0 0 0>, rot=<0 0 0 1>}
 dyn_r_sleeve_front_master {parent=upperarm_r, loc=<-13.385 -5.7245 1.2136>, rot=<-0.0111 -0.709 0.6997 -0.087>}
 dyn_r_sleeve_front {parent=dyn_r_sleeve_front_master, loc=<0 0 0>, rot=<0 0 0 -1>}
 dyn_r_sleeve_bot_master {parent=upperarm_r, loc=<-14.3441 0.2819 5.0457>, rot=<0.0114 -0.0073 0.9997 -0.0179>}
 dyn_r_sleeve_bot {parent=dyn_r_sleeve_bot_master, loc=<0 0 0>, rot=<0 0 0 1>}
 dyn_r_sleeve_top_master {parent=upperarm_r, loc=<-12.9108 0.9496 -3.5579>, rot=<-0.0233 -0.9914 -0.0348 -0.1241>}
 dyn_r_sleeve_top {parent=dyn_r_sleeve_top_master, loc=<0 0 0>, rot=<0 0 0 1>}
 FX_Shoulder_R1 {parent=upperarm_r, type=Null, loc=<-15.3817 0.0003 -3.9903>, rot=<-1 0 0 0>}
 deform_pectoral_fr_r {parent=clavicle_r, loc=<-9.39 0 0>, rot=<0 0 0 1>}
 FX_Shoulder_R2 {parent=clavicle_r, type=Null, loc=<-10.4113 1.8216 -5.5566>, rot=<0 -0.9659 0 0.2588>}
 neck_01 {parent=spine_05, loc=<11.3138 0.1529 0>, rot=<0 0 -0.2003 -0.9797>}
 neck_02 {parent=neck_01, loc=<4.5627 0 0>, rot=<0 0 0.0504 -0.9987>}
 head {parent=neck_02, loc=<5.3914 0 0>, rot=<0 0 0.115 -0.9934>}
 faceAttach {parent=head, loc=<0.6201 0.9296 0>, rot=<0.0082 -0.7071 -0.0082 -0.7071>, scale=<0.89 0.89 0.89>}
 L_cheek_inner {parent=faceAttach, loc=<4.7735 7.9774 -0.2768>, rot=<-0.0576 -0.2367 0.3228 -0.9146>}
 L_lip_corner {parent=faceAttach, loc=<2.4784 8.7831 -2.7803>, rot=<0.0023 0.0087 -0.0145 -0.9999>}
 R_cheek_inner {parent=faceAttach, loc=<-4.7735 7.9774 -0.2768>, rot=<-0.0592 0.2409 -0.3223 -0.9136>}
 R_lip_corner {parent=faceAttach, loc=<-2.4596 8.7919 -2.771>, rot=<0.0027 -0.0082 0.0146 -0.9999>}
 C_jaw {parent=faceAttach, loc=<-0.0066 1.7529 -1.6073>, rot=<0 0 0 1>}
 teeth_lower {parent=C_jaw, loc=<0.0066 7.0975 -2.309>, rot=<0 0 0 1>}
 tongue {parent=C_jaw, loc=<0.0075 2.0773 -0.9185>, rot=<-0.1568 0 0 -0.9876>}
 L_lip_lower_outer {parent=C_jaw, loc=<1.5353 7.3316 -1.1462>, rot=<0.0269 0.1204 0.3617 -0.9241>}
 C_lip_lower_mid {parent=C_jaw, loc=<-0.0009 7.9969 -1.4981>, rot=<0 -0.0236 -0.005 -0.9997>}
 R_lip_lower_outer {parent=C_jaw, loc=<-1.5164 7.3996 -1.1519>, rot=<0.0172 -0.0673 -0.3228 -0.9439>}
 C_brow_mid {parent=faceAttach, loc=<-0.0006 9.6609 4.6746>, rot=<0 0.0339 -0.0067 -0.9994>}
 L_brow_mid {parent=faceAttach, loc=<3.1938 9.3067 5.7657>, rot=<-0.0002 -0.0052 0.2553 -0.9669>}
 L_brow_outer {parent=faceAttach, loc=<5.6033 7.4667 5.6696>, rot=<0.3155 0.3938 0.3057 -0.8074>}
 C_lip_upper_mid {parent=faceAttach, loc=<0 9.9771 -2.5515>, rot=<0.2225 0.0014 -0.0005 -0.9749>}
 L_lip_upper_outer {parent=faceAttach, loc=<1.5354 9.2021 -2.7598>, rot=<0.0298 -0.0899 0.2691 -0.9584>}
 C_nose_bridge {parent=faceAttach, loc=<0.0015 10.8369 1.7604>, rot=<-0.2218 -0.0037 -0.0011 -0.9751>}
 R_brow_mid {parent=faceAttach, loc=<-3.1937 9.3067 5.7658>, rot=<-0.0001 0.0043 -0.2553 -0.9669>}
 R_brow_outer {parent=faceAttach, loc=<-5.6032 7.4666 5.6696>, rot=<0.0054 -0.0029 -0.4416 -0.8972>}
 R_eye {parent=faceAttach, loc=<-3.0141 7.0643 4.1496>, rot=<0 0 0.9998 -0.0187>}
 R_eye_lid_upper_mid {parent=R_eye, loc=<-0.0396 -1.1714 0.2631>, rot=<0.0523 -0.0009 0.0167 -0.9985>}
 R_eye_lid_lower_mid {parent=R_eye, loc=<-0.0085 -0.9425 -0.7051>, rot=<0.0329 -0.0005 0.0167 -0.9993>}
 L_eye {parent=faceAttach, loc=<3.0144 7.0636 4.1499>, rot=<0 0 0.9998 0.0187>}
 L_eye_lid_upper_mid {parent=L_eye, loc=<0.0251 -1.1623 0.2388>, rot=<0.0523 0 0 -0.9986>}
 L_eye_lid_lower_mid {parent=L_eye, loc=<0.0223 -0.923 -0.7488>, rot=<0.0329 0 0 -0.9995>}
 teeth_upper {parent=faceAttach, loc=<0.0008 8.3009 -1.4016>, rot=<0 0 0 1>}
 R_lip_upper_outer {parent=faceAttach, loc=<-1.5356 9.2014 -2.7597>, rot=<0.0305 0.0881 -0.2684 -0.9588>}
 dyn_backFix {parent=spine_05, loc=<0 0 0>, rot=<0 0 0 1>}
 attach_cape {parent=spine_05, loc=<9.017 -11.714 0>, rot=<0.0616 -0.7044 -0.0616 -0.7044>}
 attach_backpack {parent=spine_05, loc=<-0.123 -13.473 0>, rot=<-0.0093 -0.707 0.0093 -0.707>}
 Backpack_BR {parent=spine_05, type=Null, loc=<-0.1228 -10.0371 0>, rot=<0 0.7071 0 0.7071>}
 cape {parent=spine_05, type=Null, loc=<9.0544 -12.1734 0.0365>, rot=<-0.0945 0.7038 0.0377 0.7031>}
 deform_pec_aim_base_l {parent=spine_04, loc=<7.8732 12.6388 -2.4123>, rot=<0.0102 -0.4602 0.3979 -0.7936>}
 deform_pec_aim_l {parent=deform_pec_aim_base_l, loc=<0 0 0>, rot=<0 0 0 1>}
 deform_pec_base_l {parent=deform_pec_aim_l, loc=<10.766 0 0>, rot=<0 0 0 1>}
 deform_pec_l {parent=deform_pec_base_l, loc=<0.055 0 0>, rot=<0 0 0 1>}
 deform_pec_aim_base_r {parent=spine_04, loc=<7.8727 12.6388 2.4123>, rot=<-0.4602 -0.0102 0.7936 0.3979>}
 deform_pec_aim_r {parent=deform_pec_aim_base_r, loc=<0 0 0>, rot=<0 0 0 1>}
 deform_pec_base_r {parent=deform_pec_aim_r, loc=<-10.7662 -0.0006 -0.0007>, rot=<0 0 0 1>}
 deform_pec_r {parent=deform_pec_base_r, loc=<0 0 0>, rot=<0 0 0 1>}
 dyn_jacket_master {parent=spine_02, loc=<6.0104 0 0>, rot=<0 0 0.0149 -0.9999>}
 dyn_jacket_fr_r_master {parent=dyn_jacket_master, loc=<-2.0984 8.1961 8.8766>, rot=<0.0116 0.4962 0.8673 -0.037>}
 dyn_jacket_fr_r_01 {parent=dyn_jacket_fr_r_master, loc=<0 0 0>, rot=<0 0 0 1>}
 dyn_jacket_fr_r_02 {parent=dyn_jacket_fr_r_01, loc=<10.9721 0 0>, rot=<0.0151 -0.0282 -0.0083 -0.9995>}
 dyn_zipper_master {parent=dyn_jacket_fr_r_02, loc=<10.7659 -4.4609 1.0139>, rot=<-0.0712 0.0583 -0.0349 -0.9951>}
 dyn_zipper_01 {parent=dyn_zipper_master, loc=<0 0 0>, rot=<0 0 0 1>}
 dyn_jacket_r_master {parent=dyn_jacket_master, loc=<-1.7076 1.6749 13.9674>, rot=<0.033 0.0018 0.998 -0.0547>}
 dyn_jacket_r_01 {parent=dyn_jacket_r_master, loc=<0 0 0>, rot=<0 0 0 1>}
 dyn_jacket_r_02 {parent=dyn_jacket_r_01, loc=<10.3823 0 0>, rot=<0 0.0305 0 -0.9995>}
 dyn_jacket_bk_r_master {parent=dyn_jacket_master, loc=<-0.2939 -7.9416 10.5386>, rot=<0.0245 -0.4357 0.8982 -0.0527>}
 dyn_jacket_bk_r_01 {parent=dyn_jacket_bk_r_master, loc=<0 0 0>, rot=<0 0 0 1>}
 dyn_jacket_bk_r_02 {parent=dyn_jacket_bk_r_01, loc=<10.9369 0 0>, rot=<0.0273 -0.007 0.0114 -0.9995>}
 dyn_jacket_fr_l_master {parent=dyn_jacket_master, loc=<-2.0987 8.1961 -8.8766>, rot=<0.036 -0.9005 -0.4331 -0.0143>}
 dyn_jacket_fr_l_01 {parent=dyn_jacket_fr_l_master, loc=<0 0 0>, rot=<0 0 0 1>}
 dyn_jacket_fr_l_02 {parent=dyn_jacket_fr_l_01, loc=<10.9715 0 0>, rot=<0.0525 -0.0275 0.0104 -0.9982>}
 dyn_jacket_l_master {parent=dyn_jacket_master, loc=<-1.7075 1.6749 -13.9674>, rot=<0.0547 -0.998 -0.0018 -0.033>}
 dyn_jacket_l_01 {parent=dyn_jacket_l_master, loc=<0 0 0>, rot=<0 0 0 1>}
 dyn_jacket_l_02 {parent=dyn_jacket_l_01, loc=<10.3826 0 0>, rot=<0 0.0305 0 -0.9995>}
 dyn_jacket_bk_l_master {parent=dyn_jacket_master, loc=<-0.2939 -7.9416 -10.5386>, rot=<0.053 -0.9032 0.4252 -0.0239>}
 dyn_jacket_bk_l_01 {parent=dyn_jacket_bk_l_master, loc=<0 0 0>, rot=<0 0 0 1>}
 dyn_jacket_bk_l_02 {parent=dyn_jacket_bk_l_01, loc=<10.9373 0 0>, rot=<-0.0029 -0.0065 -0.0117 -0.9999>}
 dyn_jacket_bk_master {parent=dyn_jacket_master, loc=<0.3091 -10.7137 0.0218>, rot=<0.0418 -0.7059 0.7059 -0.0404>}
 dyn_jacket_bk_01 {parent=dyn_jacket_bk_master, loc=<0 0 0>, rot=<0 0 0 1>}
 dyn_jacket_bk_02 {parent=dyn_jacket_bk_01, loc=<10.9373 0 0>, rot=<-0.0002 0.025 -0.0049 -0.9997>}
 FX_Hip_Upper_L {parent=spine_01, type=Null, loc=<1.5313 0 -12.7999>, rot=<-1 0 0 0>}
 FX_Hip_Upper_R {parent=spine_01, type=Null, loc=<1.5313 0 12.7999>, rot=<0 0 0 1>}
 thigh_l {parent=pelvis, loc=<-2.5693 -0.0109 -7.7501>, rot=<0.0012 0.0572 0.0155 -0.9982>}
 calf_l {parent=thigh_l, loc=<-39.0484 0 0>, rot=<0.0006 -0.0161 -0.0391 -0.9991>}
 calf_twist_01_l {parent=calf_l, loc=<-35.813 -0.0967 0.0397>, rot=<0.0237 0.0095 -0.0035 -0.9997>}
 Corrective_foot_l {parent=calf_twist_01_l, loc=<-2.6 -2.6 0>, rot=<-0.0247 -0.05 0.0414 -0.9976>}
 calf_twist_02_l {parent=calf_l, loc=<-18.8064 -0.0967 0.0397>, rot=<0.0237 0.0095 -0.0035 -0.9997>}
 foot_l {parent=calf_l, loc=<-42.561 0.1077 0.4292>, rot=<0.0418 -0.0455 0.0343 -0.9975>}
 ball_l {parent=foot_l, loc=<-6.1598 13.8261 0.6806>, rot=<-0.0001 0.0001 -0.7062 -0.708>}
 deform_knee_in_base_l {parent=calf_l, loc=<1.298 -5.2957 -0.0496>, rot=<0 0 0 1>}
 deform_knee_in_l {parent=deform_knee_in_base_l, loc=<0 0 0>, rot=<0 0 0 1>}
 deform_knee_out_base_l {parent=calf_l, loc=<0 4 0>, rot=<0 0 0 1>}
 deform_knee_out_l {parent=deform_knee_out_base_l, loc=<0 0 0>, rot=<0 0 0 1>}
 FX_Pants_L3 {parent=calf_l, type=Null, loc=<-31.7312 -0.4337 -4.0635>, rot=<-1 0 0 0>}
 FX_Pants_L2 {parent=calf_l, type=Null, loc=<0 -0.1425 -3.7705>, rot=<-1 0 0 0>}
 thigh_twist_01_l {parent=thigh_l, loc=<-21.6303 0 0>, rot=<0.0013 0 0 -1>}
 deform_groin_base_l {parent=thigh_l, loc=<-2.1908 8.1142 -1.7442>, rot=<0 0 0 -1>}
 deform_groin_l {parent=deform_groin_base_l, loc=<0 0 0>, rot=<0 0 0 1>}
 FX_Pants_L1 {parent=thigh_l, type=Null, loc=<-7.9737 -0.0001 -6.8895>, rot=<0 -1 0 0>}
 thigh_r {parent=pelvis, loc=<-2.5691 -0.0109 7.7501>, rot=<0.0572 -0.0012 0.9982 0.0155>}
 calf_r {parent=thigh_r, loc=<39.0487 0 0>, rot=<0.0006 -0.0161 -0.0391 -0.9991>}
 calf_twist_01_r {parent=calf_r, loc=<35.8132 0.0967 -0.0397>, rot=<0.0237 0.0095 -0.0035 -0.9997>}
 Corrective_foot_r {parent=calf_twist_01_r, loc=<2.6 2.6 0>, rot=<0.0182 -0.0509 0.7372 -0.6735>}
 calf_twist_02_r {parent=calf_r, loc=<18.8064 0.0967 -0.0397>, rot=<0.0237 0.0095 -0.0035 -0.9997>}
 foot_r {parent=calf_r, loc=<42.5612 -0.1077 -0.4292>, rot=<0.0418 -0.0455 0.0343 -0.9975>}
 ball_r {parent=foot_r, loc=<6.1598 -13.8261 -0.6806>, rot=<-0.045 0.045 -0.7048 -0.7065>}
 deform_knee_in_base_r {parent=calf_r, loc=<-1.2981 5.2958 0.0496>, rot=<0 0 0 1>}
 deform_knee_in_r {parent=deform_knee_in_base_r, loc=<0 0 0>, rot=<0 0 0 1>}
 deform_knee_out_base_r {parent=calf_r, loc=<0 -4 0>, rot=<0 0 0 1>}
 deform_knee_out_r {parent=deform_knee_out_base_r, loc=<0 0 0>, rot=<0 0 0 1>}
 FX_Pats_R3 {parent=calf_r, type=Null, loc=<31.7312 0.4337 4.0635>, rot=<0 0 0 1>}
 FX_Pants_R2 {parent=calf_r, type=Null, loc=<0 0.1425 3.7705>, rot=<0 0 0 1>}
 thigh_twist_01_r {parent=thigh_r, loc=<21.6303 0 0>, rot=<0.0013 0 0 -1>}
 deform_groin_base_r {parent=thigh_r, loc=<2.1909 -8.1142 1.7442>, rot=<0 0 0 -1>}
 deform_groin_r {parent=deform_groin_base_r, loc=<0 0 0>, rot=<0 0 0 1>}
 FX_Pants_R1 {parent=thigh_r, type=Null, loc=<7.9737 0 6.8895>, rot=<0 0 0 1>}
 deform_glute_base_l {parent=pelvis, loc=<-8.8064 -8.8222 -8.4287>, rot=<0.0275 0.0501 -0.4472 -0.8926>}
 deform_glute_l {parent=deform_glute_base_l, loc=<0 0 0>, rot=<0 0 0 1>}
 deform_glute_base_r {parent=pelvis, loc=<-8.8062 -8.8222 8.4287>, rot=<0.0501 -0.0275 0.8926 -0.4472>}
 deform_glute_r {parent=deform_glute_base_r, loc=<0 0 0>, rot=<0 0 0 1>}
 FX_Hip_Lower_LB {parent=pelvis, type=Null, loc=<-4.1127 -11.0165 -7.7114>, rot=<0.7934 0 0 0.6088>}
 FX_Hip_Lower_LF {parent=pelvis, type=Null, loc=<-5.3077 9.2459 -7.1825>, rot=<-0.766 0 0 0.6428>}
 FX_Hip_Lower_RB {parent=pelvis, type=Null, loc=<-2.4774 -11.2324 8.3979>, rot=<0.6088 0 0 0.7934>}
 FX_Hip_Lower_RF {parent=pelvis, type=Null, loc=<-4.3996 9.2459 7.7571>, rot=<-0.6428 0 0 0.7661>}
]
variables [
 Calf_L_Delta {type=double, default=0}
 Calf_R_Delta {type=double, default=0}
 Is play emote {type=bool, default=False}
 Relativ_Pelvis_Root_Walk {type=double, default=0}
 Relativ_Pevlis_Root {type=double, default=0}
 Relavtive spine X {type=double, default=0}
 Relavtive spine z {type=double, default=0}
 SPR_Thig_Side_l {type=double, default=0}
 SPR_Thig_l {type=double, default=0}
 SPR_Thigh_Side_r {type=double, default=0}
 SPR_Thigh_r {type=double, default=0}
]
