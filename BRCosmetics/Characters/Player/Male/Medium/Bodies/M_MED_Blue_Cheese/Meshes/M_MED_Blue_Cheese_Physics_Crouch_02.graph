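# Physics asset for the M_MED_Blue_Cheese male player body's jacket, named as a crouch variant, with capsule bodies for jacket, pelvis, spine and thighs.
# Seven jacket chains hang from master bodies locked to spine_01, with angle-limited joints below; back-center and back-left chains have one segment, the others two.
# Thighs rotate freely on the pelvis, spine joints are angle-limited, and both front-left segments carry double capsules for wider coverage.
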PhysicsAsset /BRCosmetics/Characters/Player/Male/Medium/Bodies/M_MED_Blue_Cheese/Meshes/M_MED_Blue_Cheese_Physics_Crouch_02
bodies {
 dyn_jacket_bk_01 [{shapeType=Capsule, length=6.3016, radius=3.3905, center=<6.2096 0 0.0001>, rotation=<rot 90 0 0>}]
 dyn_jacket_bk_l_01 [{shapeType=Capsule, length=6.3016, radius=3.3905, center=<6.2096 0 0.0001>, rotation=<rot 90 0 0>}]
 dyn_jacket_bk_l_master [{shapeType=Capsule, length=2, radius=2, rotation=<rot 90 0 0>}]
 dyn_jacket_bk_master [{shapeType=Capsule, length=2, radius=2, rotation=<rot 90 0 0>}]
 dyn_jacket_bk_r_01 [{shapeType=Capsule, length=6.3016, radius=3.3905, center=<6.2096 0 0.0001>, rotation=<rot 90 0 0>}]
 dyn_jacket_bk_r_02 [{shapeType=Capsule, length=14.2563, radius=2.2467, center=<7.607 0 0.0001>, rotation=<rot 0 0 90>}]
 dyn_jacket_bk_r_master [{shapeType=Capsule, length=2, radius=2, rotation=<rot 90 0 0>}]
 dyn_jacket_fr_l_01 [
  {shapeType=Capsule, length=7.9558, radius=1.7362, center=<9.192 0.0748 1.0753>, rotation=<rot 0 0 93.9747>}
  {
   shapeType=Capsule
   length=7.9558
   radius=1.7362
   shapeName=dyn_jacket_fr_l_01_capsule
   center=<9.3054 6.2213 -0.0606>
   rotation=<rot 0 0 126.7114>
  }
 ]
 dyn_jacket_fr_l_02 [
  {shapeType=Capsule, length=13.444, radius=2.7835, center=<12.5228 1.074 6.8435>, rotation=<rot 0 0 98.919>}
  {
   shapeType=Capsule
   length=12.2811
   radius=2.3809
   shapeName=dyn_jacket_fr_r_02_capsule
   center=<11.4394 -9.9946 5.3579>
   rotation=<rot 0 0 56.029>
  }
 ]
 dyn_jacket_fr_l_master [{shapeType=Capsule, length=2, radius=2, rotation=<rot 90 0 0>}]
 dyn_jacket_fr_r_01 [{shapeType=Capsule, length=7.1831, radius=2.5089, center=<9.192 0.0995 1.4316>, rotation=<rot 0 0 65.7398>}]
 dyn_jacket_fr_r_02 [{shapeType=Capsule, length=9.4635, radius=3.045, center=<10.5461 0 3.3917>, rotation=<rot 0 0 90>}]
 dyn_jacket_fr_r_master [{shapeType=Capsule, length=2, radius=2, rotation=<rot 90 0 0>}]
 dyn_jacket_l_01 [{shapeType=Capsule, length=10.968, radius=2.7651, center=<8.8598 0.0001 0.0001>, rotation=<rot 0 0 90>}]
 dyn_jacket_l_02 [{shapeType=Capsule, length=13.4816, radius=3.0213, center=<9.7558 -0.0001 0.0001>, rotation=<rot 0 0 90>}]
 dyn_jacket_l_master [{shapeType=Capsule, length=2, radius=2, rotation=<rot 90 0 0>}]
 dyn_jacket_r_01 [{shapeType=Capsule, length=11.228, radius=2.5051, center=<8.878 0 0>, rotation=<rot 0 0 90>}]
 dyn_jacket_r_02 [{shapeType=Capsule, length=12.4886, radius=4.0144, center=<9.7558 -0.0001 0.0001>, rotation=<rot 0 0 90>}]
 dyn_jacket_r_master [{shapeType=Capsule, length=2, radius=2, rotation=<rot 90 0 0>}]
 pelvis [{shapeType=Capsule, length=8.8192, radius=13.819, center=<-0.0001 0 0>, rotation=<rot 90 0 0>}]
 spine_01 [
  {shapeType=Capsule, length=5.7954, radius=11.8927, center=<1.8784 0 0>, rotation=<rot 90 0 0>}
  {shapeType=Capsule, length=16.2577, radius=11.8927, shapeName=spine_01_capsule, center=<1.8251 0.0001 1.717>, rotation=<rot 90 0 0>}
 ]
 spine_03 [{shapeType=Capsule, length=0, radius=10.8824, rotation=<rot 90 0 0>}]
 thigh_l [{shapeType=Capsule, length=25.9209, radius=8.2744, center=<-19.2065 0 0>, rotation=<rot 90 0 -3.5456>}]
 thigh_r [{shapeType=Capsule, length=25.9209, radius=8.2744, center=<18.7709 -0.6755 -0.0001>, rotation=<rot 90 0 -3.5456>}]
}
constraints [
 {bone1Name=spine_01, bone2Name=pelvis, swing1LimitDegrees=45, swing2LimitDegrees=45, twistLimitDegrees=45}
 {bone1Name=spine_03, bone2Name=spine_01, swing1LimitDegrees=45, swing2LimitDegrees=45, twistLimitDegrees=45}
 {
  bone1Name=dyn_jacket_fr_r_master
  bone2Name=spine_01
  swing1LimitDegrees=45
  swing1Motion=Locked
  swing2LimitDegrees=45
  swing2Motion=Locked
  twistLimitDegrees=45
  twistMotion=Locked
 }
 {bone1Name=dyn_jacket_fr_r_01, bone2Name=dyn_jacket_fr_r_master, swing1LimitDegrees=20, swing2LimitDegrees=40, twistLimitDegrees=15}
 {
  bone1Name=dyn_jacket_r_master
  bone2Name=spine_01
  swing1LimitDegrees=45
  swing1Motion=Locked
  swing2LimitDegrees=45
  swing2Motion=Locked
  twistLimitDegrees=45
  twistMotion=Locked
 }
 {bone1Name=dyn_jacket_r_01, bone2Name=dyn_jacket_r_master, swing1LimitDegrees=20, swing2LimitDegrees=15, twistLimitDegrees=3}
 {
  bone1Name=dyn_jacket_bk_r_master
  bone2Name=spine_01
  swing1LimitDegrees=45
  swing1Motion=Locked
  swing2LimitDegrees=45
  swing2Motion=Locked
  twistLimitDegrees=45
  twistMotion=Locked
 }
 {bone1Name=dyn_jacket_bk_r_01, bone2Name=dyn_jacket_bk_r_master, swing1LimitDegrees=10, swing2LimitDegrees=20, twistLimitDegrees=5}
 {
  bone1Name=dyn_jacket_fr_l_master
  bone2Name=spine_01
  swing1LimitDegrees=45
  swing1Motion=Locked
  swing2LimitDegrees=45
  swing2Motion=Locked
  twistLimitDegrees=45
  twistMotion=Locked
 }
 {bone1Name=dyn_jacket_fr_l_01, bone2Name=dyn_jacket_fr_l_master, swing1LimitDegrees=10, swing2LimitDegrees=50, twistLimitDegrees=5}
 {
  bone1Name=dyn_jacket_l_master
  bone2Name=spine_01
  swing1LimitDegrees=45
  swing1Motion=Locked
  swing2LimitDegrees=45
  swing2Motion=Locked
  twistLimitDegrees=45
  twistMotion=Locked
 }
 {bone1Name=dyn_jacket_l_01, bone2Name=dyn_jacket_l_master, swing1LimitDegrees=20, swing2LimitDegrees=15, twistLimitDegrees=3}
 {
  bone1Name=dyn_jacket_bk_l_master
  bone2Name=spine_01
  swing1LimitDegrees=45
  swing1Motion=Locked
  swing2LimitDegrees=45
  swing2Motion=Locked
  twistLimitDegrees=45
  twistMotion=Locked
 }
 {bone1Name=dyn_jacket_bk_l_01, bone2Name=dyn_jacket_bk_l_master, swing1LimitDegrees=10, swing2LimitDegrees=20, twistLimitDegrees=5}
 {
  bone1Name=dyn_jacket_bk_master
  bone2Name=spine_01
  swing1LimitDegrees=45
  swing1Motion=Locked
  swing2LimitDegrees=45
  swing2Motion=Locked
  twistLimitDegrees=45
  twistMotion=Locked
 }
 {bone1Name=dyn_jacket_bk_01, bone2Name=dyn_jacket_bk_master, swing1LimitDegrees=10, swing2LimitDegrees=20, twistLimitDegrees=5}
 {
  bone1Name=thigh_r
  bone2Name=pelvis
  swing1LimitDegrees=45
  swing1Motion=Free
  swing2LimitDegrees=45
  swing2Motion=Free
  twistLimitDegrees=45
  twistMotion=Free
 }
 {
  bone1Name=thigh_l
  bone2Name=pelvis
  swing1LimitDegrees=45
  swing1Motion=Free
  swing2LimitDegrees=45
  swing2Motion=Free
  twistLimitDegrees=45
  twistMotion=Free
 }
 {bone1Name=dyn_jacket_fr_r_02, bone2Name=dyn_jacket_fr_r_01, swing1LimitDegrees=20, swing2LimitDegrees=40, twistLimitDegrees=15}
 {bone1Name=dyn_jacket_r_02, bone2Name=dyn_jacket_r_01, swing1LimitDegrees=10, swing2LimitDegrees=40, twistLimitDegrees=5}
 {bone1Name=dyn_jacket_bk_r_02, bone2Name=dyn_jacket_bk_r_01, swing1LimitDegrees=10, swing2LimitDegrees=20, twistLimitDegrees=10}
 {bone1Name=dyn_jacket_fr_l_02, bone2Name=dyn_jacket_fr_l_01, swing1LimitDegrees=50, swing2LimitDegrees=55, twistLimitDegrees=35}
 {bone1Name=dyn_jacket_l_02, bone2Name=dyn_jacket_l_01, swing1LimitDegrees=10, swing2LimitDegrees=40, twistLimitDegrees=5}
]
