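# Physics asset for the M_MED_Blue_Cheese male body (Stand_02), setting up secondary motion for its jacket.
# Seven two-segment capsule chains around the front, sides and back hang from master bones locked to spine_01, with narrow swing and twist limits.
# Pelvis, spine and thigh capsules form the body volume beneath the jacket.
# Free-rotation constraints cross-link two pairs of neighboring lower segments, left side-to-front and right back-to-side.
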PhysicsAsset /BRCosmetics/Characters/Player/Male/Medium/Bodies/M_MED_Blue_Cheese/Meshes/M_MED_Blue_Cheese_Physics_Stand_02
bodies {
 dyn_jacket_bk_01 [{shapeType=Capsule, length=10, radius=6.5, center=<2.083 0 5>, rotation=<rot 90 180 90>}]
 dyn_jacket_bk_02 [{shapeType=Capsule, length=10, radius=6.5, center=<5.2126 0 4.5>, rotation=<rot 0 0 90>}]
 dyn_jacket_bk_l_01 [{shapeType=Capsule, length=10, radius=6.5, center=<2.083 0 5>, rotation=<rot 90 -165.9638 104.0362>}]
 dyn_jacket_bk_l_02 [{shapeType=Capsule, length=10, radius=6.5, center=<5.2126 0 5>, rotation=<rot 0 0 90>}]
 dyn_jacket_bk_l_master [{shapeType=Capsule, length=2, radius=2, rotation=<rot 90 0 0>}]
 dyn_jacket_bk_master [{shapeType=Capsule, length=2, radius=2, rotation=<rot 90 0 0>}]
 dyn_jacket_bk_r_01 [{shapeType=Capsule, length=10, radius=6.5, center=<2.083 0 5>, rotation=<rot 90 -165.9638 104.0362>}]
 dyn_jacket_bk_r_02 [{shapeType=Capsule, length=10, radius=6.5, center=<5.2126 0 5>, rotation=<rot 0 0 90>}]
 dyn_jacket_bk_r_master [{shapeType=Capsule, length=2, radius=2, rotation=<rot 90 0 0>}]
 dyn_jacket_fr_l_01 [{shapeType=Capsule, length=12.0791, radius=1.654, center=<8.878 0 0>, rotation=<rot 0 0 90>}]
 dyn_jacket_fr_l_02 [{shapeType=Capsule, length=19.6351, radius=6.5, center=<11.8389 -3.417 5.5>, rotation=<rot 0.0003 0.0002 83.3824>}]
 dyn_jacket_fr_l_master [{shapeType=Capsule, length=2, radius=2, rotation=<rot 90 0 0>}]
 dyn_jacket_fr_r_01 [{shapeType=Capsule, length=12.0791, radius=1.654, center=<8.878 0 0>, rotation=<rot 0 0 90>}]
 dyn_jacket_fr_r_02 [{shapeType=Capsule, length=13, radius=6.5, center=<11.8389 0 6>, rotation=<rot 0 0 90>}]
 dyn_jacket_fr_r_master [{shapeType=Capsule, length=2, radius=2, rotation=<rot 90 0 0>}]
 dyn_jacket_l_01 [{shapeType=Capsule, length=12.0791, radius=1.654, center=<8.878 0 0>, rotation=<rot 0 0 90>}]
 dyn_jacket_l_02 [{shapeType=Capsule, length=19.5023, radius=6.5, center=<10.0441 0 4.5>, rotation=<rot 0 0 90>}]
 dyn_jacket_l_master [{shapeType=Capsule, length=2, radius=2, rotation=<rot 90 0 0>}]
 dyn_jacket_r_01 [{shapeType=Capsule, length=12.0791, radius=1.654, center=<8.878 0 0>, rotation=<rot 0 0 90>}]
 dyn_jacket_r_02 [{shapeType=Capsule, length=13, radius=6.5, center=<10.0441 0 5>, rotation=<rot 0 0 90>}]
 dyn_jacket_r_master [{shapeType=Capsule, length=2, radius=2, rotation=<rot 90 0 0>}]
 pelvis [{shapeType=Capsule, length=9.2579, radius=11.4288, center=<-1.4733 -0.6537 0>}]
 spine_01 [{shapeType=Capsule, length=3.8508, radius=10.8644, center=<3.5969 0 0>}]
 spine_03 [{shapeType=Capsule, length=4.4133, radius=8.3449, center=<0 -0.3365 0>}]
 thigh_l [{shapeType=Capsule, length=32.6049, radius=8.2134, center=<-21.2355 0.0611 0.4255>, rotation=<rot 90 -179.7016 -2.619>}]
 thigh_r [{shapeType=Capsule, length=27.643, radius=8.8406, center=<19.057 -0.7061 -1.059>, rotation=<rot 90 -179.7016 -177.0884>}]
}
constraints [
 {bone1Name=spine_01, bone2Name=pelvis, swing1LimitDegrees=45, swing2LimitDegrees=45, twistLimitDegrees=45}
 {bone1Name=spine_03, bone2Name=spine_01, swing1LimitDegrees=45, swing2LimitDegrees=45, twistLimitDegrees=45}
 {
  bone1Name=dyn_jacket_fr_r_master
  bone2Name=spine_01
  swing1LimitDegrees=45
  swing1Motion=Locked
  swing2LimitDegrees=45
  swing2Motion=Locked
  twistLimitDegrees=45
  twistMotion=Locked
 }
 {bone1Name=dyn_jacket_fr_r_01, bone2Name=dyn_jacket_fr_r_master, swing1LimitDegrees=5, swing2LimitDegrees=25, twistLimitDegrees=5}
 {
  bone1Name=dyn_jacket_r_master
  bone2Name=spine_01
  swing1LimitDegrees=45
  swing1Motion=Locked
  swing2LimitDegrees=45
  swing2Motion=Locked
  twistLimitDegrees=45
  twistMotion=Locked
 }
 {bone1Name=dyn_jacket_r_01, bone2Name=dyn_jacket_r_master, swing1LimitDegrees=5, swing2LimitDegrees=25, twistLimitDegrees=5}
 {
  bone1Name=dyn_jacket_bk_r_master
  bone2Name=spine_01
  swing1LimitDegrees=45
  swing1Motion=Locked
  swing2LimitDegrees=45
  swing2Motion=Locked
  twistLimitDegrees=45
  twistMotion=Locked
 }
 {bone1Name=dyn_jacket_bk_r_01, bone2Name=dyn_jacket_bk_r_master, swing1LimitDegrees=5, swing2LimitDegrees=10, twistLimitDegrees=0}
 {
  bone1Name=dyn_jacket_fr_l_master
  bone2Name=spine_01
  swing1LimitDegrees=45
  swing1Motion=Locked
  swing2LimitDegrees=45
  swing2Motion=Locked
  twistLimitDegrees=45
  twistMotion=Locked
 }
 {bone1Name=dyn_jacket_fr_l_01, bone2Name=dyn_jacket_fr_l_master, swing1LimitDegrees=5, swing2LimitDegrees=25, twistLimitDegrees=5}
 {
  bone1Name=dyn_jacket_l_master
  bone2Name=spine_01
  swing1LimitDegrees=45
  swing1Motion=Locked
  swing2LimitDegrees=45
  swing2Motion=Locked
  twistLimitDegrees=45
  twistMotion=Locked
 }
 {bone1Name=dyn_jacket_l_01, bone2Name=dyn_jacket_l_master, swing1LimitDegrees=15, swing2LimitDegrees=25, twistLimitDegrees=5}
 {
  bone1Name=dyn_jacket_bk_l_master
  bone2Name=spine_01
  swing1LimitDegrees=45
  swing1Motion=Locked
  swing2LimitDegrees=45
  swing2Motion=Locked
  twistLimitDegrees=45
  twistMotion=Locked
 }
 {bone1Name=dyn_jacket_bk_l_01, bone2Name=dyn_jacket_bk_l_master, swing1LimitDegrees=15, swing2LimitDegrees=10, twistLimitDegrees=0}
 {
  bone1Name=dyn_jacket_bk_master
  bone2Name=spine_01
  swing1LimitDegrees=45
  swing1Motion=Locked
  swing2LimitDegrees=45
  swing2Motion=Locked
  twistLimitDegrees=45
  twistMotion=Locked
 }
 {bone1Name=dyn_jacket_bk_01, bone2Name=dyn_jacket_bk_master, swing1LimitDegrees=5, swing2LimitDegrees=10, twistLimitDegrees=0}
 {
  bone1Name=thigh_r
  bone2Name=pelvis
  swing1LimitDegrees=45
  swing1Motion=Free
  swing2LimitDegrees=45
  swing2Motion=Free
  twistLimitDegrees=45
  twistMotion=Free
 }
 {
  bone1Name=thigh_l
  bone2Name=pelvis
  swing1LimitDegrees=45
  swing1Motion=Free
  swing2LimitDegrees=45
  swing2Motion=Free
  twistLimitDegrees=45
  twistMotion=Free
 }
 {bone1Name=dyn_jacket_fr_r_02, bone2Name=dyn_jacket_fr_r_01, swing1LimitDegrees=5, swing2LimitDegrees=25, twistLimitDegrees=5}
 {bone1Name=dyn_jacket_r_02, bone2Name=dyn_jacket_r_01, swing1LimitDegrees=5, swing2LimitDegrees=25, twistLimitDegrees=5}
 {bone1Name=dyn_jacket_bk_r_02, bone2Name=dyn_jacket_bk_r_01, swing1LimitDegrees=5, swing2LimitDegrees=25, twistLimitDegrees=5}
 {bone1Name=dyn_jacket_fr_l_02, bone2Name=dyn_jacket_fr_l_01, swing1LimitDegrees=5, swing2LimitDegrees=15, twistLimitDegrees=5}
 {bone1Name=dyn_jacket_l_02, bone2Name=dyn_jacket_l_01, swing1LimitDegrees=15, swing2LimitDegrees=25, twistLimitDegrees=5}
 {bone1Name=dyn_jacket_bk_02, bone2Name=dyn_jacket_bk_01, swing1LimitDegrees=5, swing2LimitDegrees=25, twistLimitDegrees=5}
 {bone1Name=dyn_jacket_bk_l_02, bone2Name=dyn_jacket_bk_l_01, swing1LimitDegrees=15, swing2LimitDegrees=25, twistLimitDegrees=5}
 {
  bone1Name=dyn_jacket_l_02
  bone2Name=dyn_jacket_fr_l_02
  swing1LimitDegrees=45
  swing1Motion=Free
  swing2LimitDegrees=45
  swing2Motion=Free
  twistLimitDegrees=45
  twistMotion=Free
 }
 {
  bone1Name=dyn_jacket_bk_r_02
  bone2Name=dyn_jacket_r_02
  swing1LimitDegrees=45
  swing1Motion=Free
  swing2LimitDegrees=45
  swing2Motion=Free
  twistLimitDegrees=45
  twistMotion=Free
 }
]
